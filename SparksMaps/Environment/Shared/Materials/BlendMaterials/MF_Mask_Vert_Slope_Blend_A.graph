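# Material function that blends a Top Layer over a Bottom Layer of material attributes using texture, vertex color and slope masks.
# The cutoff-scaled mask, with optional exclusion, gains noise or mask falloff before optional vertex color and WorldAlignedBlend slope weighting.
# A packed four-value remap tunes the edge and opacity scales the final alpha.
# Switches offer overlay, hard light or color dodge color blends and blending of underlying normals into the top.
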MaterialFunction /SparksMaps/Environment/Shared/Materials/BlendMaterials/MF_Mask_Vert_Slope_Blend_A
graph {nodes [
  n1 = GetMaterialAttributes(None=FunctionInput() {inputName=Bottom Layer, inputType=FunctionInput_MaterialAttributes}) {attributeGetTypes [#1, #2]}
  n2 = GetMaterialAttributes(None=FunctionInput() {inputName=Top Layer, inputType=FunctionInput_MaterialAttributes}) {attributeGetTypes [#1, #2]}
  n3 = OneMinus(None=FunctionInput(Preview=Constant()) {inputName=Underlying Normal Intensity, inputType=FunctionInput_Scalar})
  n4 = StaticSwitch(
   True=MaterialFunctionCall(
    BaseNormal=MaterialFunctionCall(Normal=@n1.Normal, Flatness=@n3) {materialFunction=&FlattenNormal}.Result
    AdditionalNormal=@n2.Normal
   ) {materialFunction=&BlendAngleCorrectedNormals}.Result
   False=@n2.Normal
   Value=FunctionInput(Preview=StaticBool()) {inputName=Edit Underlying Normal Intensity, inputType=FunctionInput_StaticBool}
  )
  n5 = NamedRerouteUsage()
  n6 = FunctionInput(Preview=Constant4Vector() {constant=<rgba -15 150 -15 120>}.RGBA) {inputName=Fine Tune Edge, inputType=FunctionInput_Vector4}
  n7 = MaterialFunctionCall!(
   In Explicit Normal=NamedRerouteUsage()
   In World Vector
   Blend Sharpness=FunctionInput(Preview=Constant() {r=10}) {inputName=Slope Blend Sharpness, inputType=FunctionInput_Scalar}
   Blend Bias=FunctionInput(Preview=Constant()) {inputName=Slope Blend Strength, inputType=FunctionInput_Scalar}
   Clamped?
   Alpha
  ) {materialFunction=/Engine/Functions/Engine_MaterialFunctions01/AlphaBlend/WorldAlignedBlend}
  n8 = OneMinus(None=FunctionInput(Preview=Constant() {r=1}) {inputName=Slope Intensity, inputType=FunctionInput_Scalar})
  n9 = Noise!(World Position, FilterWidth) {bTiling=true, bTurbulence=false, noiseFunction=NOISEFUNCTION_GradientALU, outputMin=0, scale=0.1}
  n10 = FunctionInput(Preview=Constant() {r=1}) {inputName=Falloff Noise Intensity, inputType=FunctionInput_Scalar}
  n11 = StaticBool()
  n12 = TextureSample() {texture=/Game/Packages/Fortress_Transylvania/Haunted_Forest/Swamps/Textures/2ndInfectedMapLinear2}
  n13 = FunctionInput(Preview=@n12.RGB) {inputName=Mask Channel, inputType=FunctionInput_Scalar}
  n14 = FunctionInput(Preview=@n12.RGB) {inputName=Exclusion Mask, inputType=FunctionInput_Scalar}
  n15 = StaticBool()
  n16 = Multiply(
   A=FunctionInput(Preview=Constant() {r=1}) {inputName=Cutoff, inputType=FunctionInput_Scalar}
   B=StaticSwitch(
    True=Subtract(A=@n13, B=@n14)
    False=@n13
    Value=FunctionInput(Preview=@n15) {inputName=Use Exclusion Mask, inputType=FunctionInput_StaticBool}
   )
  )
  n17 = Add(
   A=StaticSwitch(
    True=NamedRerouteUsage()
    False=LinearInterpolate(B=@n9, Alpha=@n10)
    Value=FunctionInput(Preview=@n11) {inputName=Use Mask for Vertex Falloff, inputType=FunctionInput_StaticBool}
   ) {defaultValue=true}
   B=@n16
  )
  n18 = VertexColor()
  n19 = StaticSwitch(
   True=Multiply(A=Saturate(None=@n17), B=FunctionInput(Preview=@n18.R) {inputName=Vertex Color Channel, inputType=FunctionInput_Scalar})
   False=@n16
   Value=FunctionInput(Preview=StaticBool() {value=true}) {inputName=Use Vertex Color Mask, inputType=FunctionInput_StaticBool}
  )
  n20 = Multiply(A=@n19, B=Saturate(None=Add(A=@n7.w/Vertex Normals, B=@n8)))
  n21 = FunctionInput(Preview=StaticBool()) {inputName=Use Slope Mask, inputType=FunctionInput_StaticBool}
  n22 = Saturate(None=LinearInterpolate(
    A=ComponentMask(None=@n6) {b=true, r=true}
    B=ComponentMask(None=@n6) {a=true, g=true}
    Alpha=StaticSwitch(True=@n20, False=@n19, Value=@n21)
   ))
  n23 = ComponentMask(None=@n22) {r=true}
  n24 = Subtract(A=@n23, B=ComponentMask(None=@n22) {g=true})
  n25 = FunctionInput(Preview=Constant() {r=3}) {inputName="Edge Intensity (Inverse)", inputType=FunctionInput_Scalar}
  n26 = Multiply(
   A=FunctionInput(Preview=Constant() {r=1}) {inputName=Opacity, inputType=FunctionInput_Scalar}
   B=Subtract(A=@n23, B=Divide(A=@n24, B=@n25) {constB=3})
  )
  FunctionOutput(None=BlendMaterialAttributes(
    A=SetMaterialAttributes!(MaterialAttributes=@n1.MaterialAttributes, Normal=@n1.MaterialAttributes, BaseColor) {attributeSetTypes [#1, #2]}
    B=SetMaterialAttributes(MaterialAttributes=@n2.MaterialAttributes, Normal=@n4, BaseColor=@n5) {attributeSetTypes [#1, #2]}
    Alpha=Saturate(None=@n26)
   ))
  NamedRerouteDeclaration(None=FunctionInput(Preview=@n12.RGB) {inputName=Vertex Falloff Mask Channel, inputType=FunctionInput_Scalar}) {name=Vertex Falloff Mask, nodeColor=<rgba 0.3359 0 1 1>}
  NamedRerouteDeclaration(None=@n1.Normal) {name=Normal Map, nodeColor=<rgba 1 0 0.7266 1>}
  NamedRerouteDeclaration(None=@n2.BaseColor) {name=Top Layer Base Color, nodeColor=<rgba 0 0.1797 1 1>}
  NamedRerouteDeclaration(None=@n1.BaseColor) {name=Bottom Layer Base Color, nodeColor=<rgba 0.5 1 0 1>}
  n27 = NamedRerouteUsage()
  n28 = NamedRerouteUsage()
  n29 = MaterialFunctionCall!(Base=@n27, Blend=@n28) {materialFunction=/Engine/Functions/Engine_MaterialFunctions03/Blends/Blend_HardLight}
  n30 = StaticSwitch(
   True=MaterialFunctionCall(Base=@n27, Blend=@n28) {materialFunction=&Blend_Overlay}.Result
   False=NamedRerouteUsage()
   Value=FunctionInput(Preview=StaticBool()) {inputName=Overlay Blend, inputType=FunctionInput_StaticBool}
  )
  n31 = FunctionInput(Preview=StaticBool()) {inputName=Hard Light Blend, inputType=FunctionInput_StaticBool}
  n32 = StaticBool()
  NamedRerouteDeclaration(None=StaticSwitch(
    True=MaterialFunctionCall!(Blend=@n27, Base=@n28) {materialFunction=/Engine/Functions/Engine_MaterialFunctions03/Blends/Blend_ColorDodge}.Result
    False=StaticSwitch(True=@n29.Result, False=@n30, Value=@n31)
    Value=FunctionInput(Preview=@n32) {inputName=Color Dodge Blend, inputType=FunctionInput_StaticBool}
   )) {name=Blend Modes Output, nodeColor=<rgba 1 0 0.8203 1>}
 ]}
