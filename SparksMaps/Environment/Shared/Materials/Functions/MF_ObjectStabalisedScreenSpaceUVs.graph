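# Material function that outputs screen-space UVs stabilised to the object, so a projected pattern follows it on screen.
# Aspect-corrected viewport UVs are offset by the object's clip-space position, with Y flipped, to anchor them to the object's origin.
# ScaleUVsByCenter then scales them by object radius over view depth, tying pattern size to the object's apparent size.
# A scalar UV Scale input multiplies the final result.
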MaterialFunction /SparksMaps/Environment/Shared/Materials/Functions/MF_ObjectStabalisedScreenSpaceUVs
graph {nodes [
  n1 = ConstantBiasScale(None=ScreenPosition().ViewportUV) {bias=-0.5, scale=2}
  n2 = ViewSize()
  n3 = AppendVector(A=Constant() {r=1}, B=Divide(A=ComponentMask(None=@n2) {g=true}, B=ComponentMask(None=@n2) {r=true}))
  n4 = Multiply(
   A=MaterialFunctionCall(World Position=ObjectPositionWS()) {materialFunction=&TransformToClipSpace}.Clip Space XY
   B=Constant2Vector() {g=-1, r=1}
  )
  n5 = Subtract(A=ConstantBiasScale(None=Multiply(A=@n1, B=@n3)), B=Multiply(A=Multiply(A=@n3, B=@n4)) {constB=0.5})
  n6 = TransformPosition(None=ObjectPositionWS()) {transformSourceType=TRANSFORMPOSSOURCE_World, transformType=TRANSFORMPOSSOURCE_View}
  n7 = Multiply(A=Divide(A=Constant() {r=1}, B=ComponentMask(None=@n6) {b=true}), B=ObjectRadius())
  n8 = Constant() {r=100}
  FunctionOutput(None=Multiply(
    A=MaterialFunctionCall(UVs=@n5, Texture Scale=@n7) {materialFunction=&ScaleUVsByCenter}.UVs
    B=FunctionInput(Preview=@n8) {inputName=UV Scale, inputType=FunctionInput_Scalar}
   ))
 ]}
