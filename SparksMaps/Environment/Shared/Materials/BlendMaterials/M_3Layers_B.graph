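# Environment material for Sparks maps that blends a base surface with two layers, Blend A and Blend B.
# Each blend uses a mask function driven by texture channels, vertex color and slope, with switchable UV sets and world-aligned projection.
# Emissive combines Fresnel, stage colors and beat pulsing, plus glitter using curve-atlas gradients and screen-space UVs.
# A shared WPO function drives vertex offset.
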Material /SparksMaps/Environment/Shared/Materials/BlendMaterials/M_3Layers_B
graph {
 expressions {
  MaterialExpressionAdd_0 {
   class=MaterialExpressionAdd
   inputs [
    A {from=MaterialExpressionMultiply_49}
    B {from=MaterialExpressionMultiply_11}
   ]
  }
  MaterialExpressionAdd_1 {
   class=MaterialExpressionAdd
   inputs [
    A {from=MaterialExpressionVertexNormalWS_0}
    B {from=MaterialExpressionNormalize_2}
   ]
  }
  MaterialExpressionAdd_10 {
   class=MaterialExpressionAdd
   inputs [
    A {from=MaterialExpressionNamedRerouteUsage_54}
    B {from=MaterialExpressionConstant_8}
   ]
  }
  MaterialExpressionAdd_11 {
   class=MaterialExpressionAdd
   inputs [
    A {from=MaterialExpressionAdd_7}
    B {from=MaterialExpressionDivide_2}
   ]
  }
  MaterialExpressionAdd_12 {
   class=MaterialExpressionAdd
   inputs [
    A {from=MaterialExpressionNamedRerouteUsage_66}
    B {from=MaterialExpressionNamedRerouteUsage_67}
   ]
  }
  MaterialExpressionAdd_13 {
   class=MaterialExpressionAdd
   inputs [
    A {from=MaterialExpressionAdd_12}
    B {from=MaterialExpressionNamedRerouteUsage_68}
   ]
  }
  MaterialExpressionAdd_14 {
   class=MaterialExpressionAdd
   inputs [
    A {from=MaterialExpressionStaticSwitch_30}
    B {from=MaterialExpressionStaticSwitch_30}
   ]
  }
  MaterialExpressionAdd_15 {
   class=MaterialExpressionAdd
   inputs [
    A {from=MaterialExpressionAdd_14}
    B {from=MaterialExpressionAdd_14}
   ]
  }
  MaterialExpressionAdd_17 {
   class=MaterialExpressionAdd
   inputs [
    A {from=MaterialExpressionAdd_15}
    B {from=MaterialExpressionAdd_15}
   ]
  }
  MaterialExpressionAdd_18 {
   class=MaterialExpressionAdd
   inputs [
    A {from=MaterialExpressionNamedRerouteUsage_72}
    B {from=MaterialExpressionStaticSwitch_24}
   ]
  }
  MaterialExpressionAdd_19 {
   class=MaterialExpressionAdd
   inputs [
    A {from=MaterialExpressionAdd_5}
    B {from=MaterialExpressionMultiply_28}
   ]
  }
  MaterialExpressionAdd_2 {
   class=MaterialExpressionAdd
   inputs [
    A {from=MaterialExpressionTextureSample_9, output=R}
    B {from=MaterialExpressionTextureSample_9, output=R}
   ]
  }
  MaterialExpressionAdd_3 {
   class=MaterialExpressionAdd
   inputs [
    A {from=MaterialExpressionAdd_2}
    B {from=MaterialExpressionTextureSample_9, output=B}
   ]
  }
  MaterialExpressionAdd_5 {
   class=MaterialExpressionAdd
   inputs [
    A {from=MaterialExpressionMaterialFunctionCall_1, output=Result}
    B {from=MaterialExpressionStaticSwitch_11}
   ]
  }
  MaterialExpressionAdd_6 {
   class=MaterialExpressionAdd
   inputs [
    A {from=MaterialExpressionNamedRerouteUsage_50}
    B {from=MaterialExpressionConstant2Vector_0}
   ]
   properties {constB=0.2}
  }
  MaterialExpressionAdd_7 {
   class=MaterialExpressionAdd
   inputs [
    A {from=MaterialExpressionDivide_0}
    B {from=MaterialExpressionDivide_1}
   ]
  }
  MaterialExpressionAdd_8 {
   class=MaterialExpressionAdd
   inputs [
    A {from=MaterialExpressionNamedRerouteUsage_53}
    B {from=MaterialExpressionConstant_6}
   ]
  }
  MaterialExpressionAdd_9 {
   class=MaterialExpressionAdd
   inputs [
    A {from=MaterialExpressionNamedRerouteUsage_51}
    B {from=MaterialExpressionConstant2Vector_1}
   ]
   properties {constB=0.2}
  }
  MaterialExpressionAtmosphericLightVector_0 {class=MaterialExpressionAtmosphericLightVector, inputs [], properties {}}
  MaterialExpressionAtmosphericLightVector_1 {class=MaterialExpressionAtmosphericLightVector, inputs [], properties {}}
  MaterialExpressionCameraVectorWS_0 {class=MaterialExpressionCameraVectorWS, inputs [], properties {}}
  MaterialExpressionChannelMaskParameter_0 {
   class=MaterialExpressionChannelMaskParameter
   inputs [None {from=MaterialExpressionTextureSample_6, output=RGB}]
   properties {group=Blend A _Mask, maskChannel=Blue, parameterName="Vertex Falloff Channel  (Blend A)", defaultValue=<rgba 0 0 1 0>}
  }
  MaterialExpressionChannelMaskParameter_1 {
   class=MaterialExpressionChannelMaskParameter
   inputs [None {from=MaterialExpressionNamedRerouteDeclaration_10, output=""}]
   properties {group=Blend A _Mask, maskChannel=Blue, parameterName="Exclusion Mask Channel (Blend A)", defaultValue=<rgba 0 0 1 0>}
  }
  MaterialExpressionChannelMaskParameter_2 {
   class=MaterialExpressionChannelMaskParameter
   inputs [None {from=MaterialExpressionVertexColor_0, output=""}]
   properties {group=Blend A _Mask, parameterName="Vertex Color Channel (Blend A)"}
  }
  MaterialExpressionChannelMaskParameter_3 {
   class=MaterialExpressionChannelMaskParameter
   inputs [None {from=MaterialExpressionTextureSample_6, output=RGB}]
   properties {group=Blend A _Mask, maskChannel=Blue, parameterName="Mask Channel (Blend A)", defaultValue=<rgba 0 0 1 0>}
  }
  MaterialExpressionChannelMaskParameter_4 {
   class=MaterialExpressionChannelMaskParameter
   inputs [None {from=MaterialExpressionStaticSwitchParameter_11, output=""}]
   properties {group=Blend B _Mask, maskChannel=Blue, parameterName="Exclusion Mask Channel (Blend B)", defaultValue=<rgba 0 0 1 0>}
  }
  MaterialExpressionChannelMaskParameter_5 {
   class=MaterialExpressionChannelMaskParameter
   inputs [None {from=MaterialExpressionStaticSwitchParameter_11, output=""}]
   properties {group=Blend B _Mask, parameterName="Mask Channel (Blend B)"}
  }
  MaterialExpressionChannelMaskParameter_6 {
   class=MaterialExpressionChannelMaskParameter
   inputs [None {from=MaterialExpressionVertexColor_3, output=""}]
   properties {group=Blend B _Mask, maskChannel=Green, parameterName="Vertex Color Channel (Blend B)", defaultValue=<rgba 0 1 0 0>}
  }
  MaterialExpressionChannelMaskParameter_7 {
   class=MaterialExpressionChannelMaskParameter
   inputs [None {from=MaterialExpressionStaticSwitchParameter_11, output=""}]
   properties {group=Blend B _Mask, parameterName="Vertex Falloff Channel (Blend B)"}
  }
  MaterialExpressionChannelMaskParameter_8 {
   class=MaterialExpressionChannelMaskParameter
   inputs [None {from=MaterialExpressionTextureSampleParameter2D_8, output=RGB}]
   properties {group=06 Glitter, parameterName=Glitter Mask Channel}
  }
  MaterialExpressionComponentMask_0 {class=MaterialExpressionComponentMask, inputs [None {from=MaterialExpressionNamedRerouteUsage_62}], properties {r=true}}
  MaterialExpressionComponentMask_1 {class=MaterialExpressionComponentMask, inputs [None {from=MaterialExpressionTextureCoordinate_1}], properties {g=true}}
  MaterialExpressionComponentMask_2 {class=MaterialExpressionComponentMask, inputs [None {from=MaterialExpressionStaticSwitchParameter_3}], properties {a=true}}
  MaterialExpressionComponentMask_29 {class=MaterialExpressionComponentMask, inputs [None {from=MaterialExpressionTextureCoordinate_10}], properties {r=true}}
  MaterialExpressionComponentMask_3 {class=MaterialExpressionComponentMask, inputs [None {from=MaterialExpressionNamedRerouteUsage_64}], properties {g=true}}
  MaterialExpressionComponentMask_4 {class=MaterialExpressionComponentMask, inputs [None {from=MaterialExpressionNamedRerouteUsage_65}], properties {r=true}}
  MaterialExpressionComponentMask_6 {class=MaterialExpressionComponentMask, inputs [None {from=MaterialExpressionStaticSwitchParameter_3}], properties {b=true, g=true, r=true}}
  MaterialExpressionComponentMask_7 {class=MaterialExpressionComponentMask, inputs [None {from=MaterialExpressionStaticSwitchParameter_6}], properties {a=true}}
  MaterialExpressionComponentMask_8 {class=MaterialExpressionComponentMask, inputs [None {from=MaterialExpressionStaticSwitchParameter_6}], properties {b=true, g=true, r=true}}
  MaterialExpressionConstant2Vector_0 {class=MaterialExpressionConstant2Vector, properties {g=0.7, r=-0.5}}
  MaterialExpressionConstant2Vector_1 {class=MaterialExpressionConstant2Vector, properties {g=-0.3, r=0.6}}
  MaterialExpressionConstant3Vector_0 {class=MaterialExpressionConstant3Vector, properties {constant=<rgba 1 1 1 0>}}
  MaterialExpressionConstant3Vector_1 {class=MaterialExpressionConstant3Vector}
  MaterialExpressionConstant_0 {class=MaterialExpressionConstant}
  MaterialExpressionConstant_1 {class=MaterialExpressionConstant}
  MaterialExpressionConstant_4 {class=MaterialExpressionConstant, properties {r=0.5}}
  MaterialExpressionConstant_5 {class=MaterialExpressionConstant, properties {r=0.5}}
  MaterialExpressionConstant_6 {class=MaterialExpressionConstant, properties {r=0.2}}
  MaterialExpressionConstant_7 {class=MaterialExpressionConstant, properties {r=0.5}}
  MaterialExpressionConstant_8 {class=MaterialExpressionConstant, properties {r=0.5}}
  MaterialExpressionCurveAtlasRowParameter_4 {
   class=MaterialExpressionCurveAtlasRowParameter
   inputs [CurveTime {from=MaterialExpressionComponentMask_29, output=""}]
   properties {
    atlas=/SparksMaps/Environment/Shared/Materials/Generic/Textures/CurveAtlas/CA_Glitter
    curve=/SparksMaps/Environment/Shared/Materials/Generic/Textures/CurveAtlas/C_GlitterCurve_1
    group=06 Glitter
    parameterName=Glitter Gradient
   }
  }
  MaterialExpressionDesaturation_0 {
   class=MaterialExpressionDesaturation
   inputs [
    None {from=MaterialExpressionReroute_9, output=""}
    Fraction {from=MaterialExpressionScalarParameter_79, output=""}
   ]
   properties {}
  }
  MaterialExpressionDistance_0 {
   class=MaterialExpressionDistance
   inputs [
    A {from=MaterialExpressionFrac_0, output=""}
    B {from=MaterialExpressionConstant_4, output=""}
   ]
   properties {}
  }
  MaterialExpressionDistance_1 {
   class=MaterialExpressionDistance
   inputs [
    A {from=MaterialExpressionFrac_1, output=""}
    B {from=MaterialExpressionConstant_5, output=""}
   ]
   properties {}
  }
  MaterialExpressionDistance_2 {
   class=MaterialExpressionDistance
   inputs [
    A {from=MaterialExpressionFrac_2, output=""}
    B {from=MaterialExpressionConstant_7, output=""}
   ]
   properties {}
  }
  MaterialExpressionDivide_0 {
   class=MaterialExpressionDivide
   inputs [
    A {from=MaterialExpressionMultiply_19}
    B {from=""}
   ]
   properties {constB=3}
  }
  MaterialExpressionDivide_1 {
   class=MaterialExpressionDivide
   inputs [
    A {from=MaterialExpressionMultiply_23}
    B {from=""}
   ]
   properties {constB=3}
  }
  MaterialExpressionDivide_2 {
   class=MaterialExpressionDivide
   inputs [
    A {from=MaterialExpressionMultiply_27}
    B {from=""}
   ]
   properties {constB=3}
  }
  MaterialExpressionDotProduct_0 {
   class=MaterialExpressionDotProduct
   inputs [
    A {from=MaterialExpressionAtmosphericLightVector_0, output=""}
    B {from=MaterialExpressionMaterialFunctionCall_31, output=Reflection Vector}
   ]
   properties {}
  }
  MaterialExpressionDotProduct_1 {
   class=MaterialExpressionDotProduct
   inputs [
    A {from=MaterialExpressionAtmosphericLightVector_1, output=""}
    B {from=MaterialExpressionVertexNormalWS_1, output=""}
   ]
   properties {}
  }
  MaterialExpressionFrac_0 {class=MaterialExpressionFrac, inputs [None {from=MaterialExpressionRotator_0, output=""}], properties {}}
  MaterialExpressionFrac_1 {class=MaterialExpressionFrac, inputs [None {from=MaterialExpressionRotator_1, output=""}], properties {}}
  MaterialExpressionFrac_2 {class=MaterialExpressionFrac, inputs [None {from=MaterialExpressionRotator_2, output=""}], properties {}}
  MaterialExpressionFresnel_0 {
   class=MaterialExpressionFresnel
   inputs [
    ExponentIn {from=MaterialExpressionScalarParameter_2, output=""}
    BaseReflectFractionIn {from=MaterialExpressionScalarParameter_4, output=""}
    Normal {from="", output=""}
   ]
   properties {baseReflectFraction=0.003}
  }
  MaterialExpressionFresnel_1 {
   class=MaterialExpressionFresnel
   inputs [
    ExponentIn {from=MaterialExpressionScalarParameter_8, output=""}
    BaseReflectFractionIn {from=MaterialExpressionScalarParameter_10, output=""}
    Normal {from="", output=""}
   ]
   properties {baseReflectFraction=0.003}
  }
  MaterialExpressionLinearInterpolate_0 {
   class=MaterialExpressionLinearInterpolate
   inputs [
    A {from=MaterialExpressionNamedRerouteUsage_15}
    B {from=MaterialExpressionStaticSwitch_9}
    Alpha {from=MaterialExpressionMultiply_2}
   ]
  }
  MaterialExpressionLinearInterpolate_1 {
   class=MaterialExpressionLinearInterpolate
   inputs [
    A {from=MaterialExpressionNamedRerouteUsage_24}
    B {from=MaterialExpressionScalarParameter_25}
    Alpha {from=MaterialExpressionNamedRerouteUsage_4}
   ]
  }
  MaterialExpressionLinearInterpolate_2 {
   class=MaterialExpressionLinearInterpolate
   inputs [
    A {from=MaterialExpressionStaticSwitch_23}
    B {from=MaterialExpressionMultiply_3}
    Alpha {from=MaterialExpressionNamedRerouteDeclaration_6}
   ]
  }
  MaterialExpressionLinearInterpolate_3 {
   class=MaterialExpressionLinearInterpolate
   inputs [
    A {from=MaterialExpressionConstant3Vector_1}
    B {from=MaterialExpressionReroute_0}
    Alpha {from=MaterialExpressionPower_1}
   ]
  }
  MaterialExpressionLinearInterpolate_4 {
   class=MaterialExpressionLinearInterpolate
   inputs [
    A {from=MaterialExpressionNamedRerouteUsage_44}
    B {from=MaterialExpressionMultiply_14}
    Alpha {from=MaterialExpressionComponentMask_2}
   ]
  }
  MaterialExpressionLinearInterpolate_5 {
   class=MaterialExpressionLinearInterpolate
   inputs [
    A {from=MaterialExpressionNamedRerouteUsage_45}
    B {from=MaterialExpressionMultiply_16}
    Alpha {from=MaterialExpressionComponentMask_7}
   ]
  }
  MaterialExpressionLinearInterpolate_6 {
   class=MaterialExpressionLinearInterpolate
   inputs [
    A {from=MaterialExpressionNamedRerouteUsage_25}
    B {from=MaterialExpressionScalarParameter_16}
    Alpha {from=MaterialExpressionNamedRerouteUsage_3}
   ]
  }
  MaterialExpressionLinearInterpolate_7 {
   class=MaterialExpressionLinearInterpolate
   inputs [
    A {from=MaterialExpressionNamedRerouteDeclaration_12}
    B {from=MaterialExpressionNamedRerouteUsage_47}
    Alpha {from=MaterialExpressionScalarParameter_64}
   ]
  }
  MaterialExpressionLinearInterpolate_8 {
   class=MaterialExpressionLinearInterpolate
   inputs [
    A {from=MaterialExpressionNamedRerouteUsage_49}
    B {from=MaterialExpressionStaticSwitch_28}
    Alpha {from=MaterialExpressionMultiply_31}
   ]
  }
  MaterialExpressionLinearInterpolate_9 {
   class=MaterialExpressionLinearInterpolate
   inputs [
    A {from=MaterialExpressionReroute_8}
    B {from=MaterialExpressionAdd_18}
    Alpha {from=MaterialExpressionNamedRerouteUsage_73}
   ]
  }
  MaterialExpressionMakeMaterialAttributes_0 {
   class=MaterialExpressionMakeMaterialAttributes
   inputs [
    BaseColor {from=MaterialExpressionNamedRerouteUsage_6, output=""}
    Metallic {from=MaterialExpressionNamedRerouteUsage_7, output=""}
    Specular {from=MaterialExpressionNamedRerouteUsage_9, output=""}
    Roughness {from=MaterialExpressionNamedRerouteUsage_8, output=""}
    Anisotropy {from="", output=""}
    EmissiveColor {from=MaterialExpressionNamedRerouteUsage_14, output=""}
    Opacity {from=MaterialExpressionNamedRerouteUsage_11, output=""}
    OpacityMask {from=MaterialExpressionNamedRerouteUsage_11, output=""}
    Normal {from=MaterialExpressionNamedRerouteUsage_12, output=""}
    Tangent {from="", output=""}
    WorldPositionOffset {from=MaterialExpressionMaterialFunctionCall_14, output=Result}
    SubsurfaceColor {from="", output=""}
    ClearCoat {from="", output=""}
    ClearCoatRoughness {from="", output=""}
    AmbientOcclusion {from="", output=""}
    Refraction {from="", output=""}
    CustomizedUVs_0 {from="", output=""}
    CustomizedUVs_1 {from="", output=""}
    CustomizedUVs_2 {from="", output=""}
    CustomizedUVs_3 {from="", output=""}
    CustomizedUVs_4 {from="", output=""}
    CustomizedUVs_5 {from="", output=""}
    CustomizedUVs_6 {from="", output=""}
    CustomizedUVs_7 {from="", output=""}
    PixelDepthOffset {from="", output=""}
    ShadingModel {from="", output=""}
    Displacement {from="", output=""}
   ]
   properties {}
  }
  MaterialExpressionMakeMaterialAttributes_1 {
   class=MaterialExpressionMakeMaterialAttributes
   inputs [
    BaseColor {from=MaterialExpressionStaticSwitchParameter_9, output=""}
    Metallic {from=MaterialExpressionMaterialFunctionCall_18, output=G}
    Specular {from=MaterialExpressionMaterialFunctionCall_18, output=G}
    Roughness {from=MaterialExpressionMaterialFunctionCall_18, output=G}
    Anisotropy {from="", output=""}
    EmissiveColor {from="", output=""}
    Opacity {from="", output=""}
    OpacityMask {from="", output=""}
    Normal {from=MaterialExpressionStaticSwitchParameter_5, output=""}
    Tangent {from="", output=""}
    WorldPositionOffset {from="", output=""}
    SubsurfaceColor {from="", output=""}
    ClearCoat {from="", output=""}
    ClearCoatRoughness {from="", output=""}
    AmbientOcclusion {from="", output=""}
    Refraction {from="", output=""}
    CustomizedUVs_0 {from="", output=""}
    CustomizedUVs_1 {from="", output=""}
    CustomizedUVs_2 {from="", output=""}
    CustomizedUVs_3 {from="", output=""}
    CustomizedUVs_4 {from="", output=""}
    CustomizedUVs_5 {from="", output=""}
    CustomizedUVs_6 {from="", output=""}
    CustomizedUVs_7 {from="", output=""}
    PixelDepthOffset {from="", output=""}
    ShadingModel {from="", output=""}
    Displacement {from="", output=""}
   ]
   properties {}
  }
  MaterialExpressionMakeMaterialAttributes_2 {
   class=MaterialExpressionMakeMaterialAttributes
   inputs [
    BaseColor {from=MaterialExpressionStaticSwitchParameter_10, output=""}
    Metallic {from=MaterialExpressionMaterialFunctionCall_24, output=G}
    Specular {from=MaterialExpressionMaterialFunctionCall_24, output=G}
    Roughness {from=MaterialExpressionMaterialFunctionCall_24, output=G}
    Anisotropy {from="", output=""}
    EmissiveColor {from="", output=""}
    Opacity {from="", output=""}
    OpacityMask {from="", output=""}
    Normal {from=MaterialExpressionStaticSwitchParameter_8, output=""}
    Tangent {from="", output=""}
    WorldPositionOffset {from="", output=""}
    SubsurfaceColor {from="", output=""}
    ClearCoat {from="", output=""}
    ClearCoatRoughness {from="", output=""}
    AmbientOcclusion {from="", output=""}
    Refraction {from="", output=""}
    CustomizedUVs_0 {from="", output=""}
    CustomizedUVs_1 {from="", output=""}
    CustomizedUVs_2 {from="", output=""}
    CustomizedUVs_3 {from="", output=""}
    CustomizedUVs_4 {from="", output=""}
    CustomizedUVs_5 {from="", output=""}
    CustomizedUVs_6 {from="", output=""}
    CustomizedUVs_7 {from="", output=""}
    PixelDepthOffset {from="", output=""}
    ShadingModel {from="", output=""}
    Displacement {from="", output=""}
   ]
   properties {}
  }
  MaterialExpressionMaterialFunctionCall_0 {
   class=MaterialExpressionMaterialFunctionCall
   inputs [
    In {from=MaterialExpressionFresnel_0}
    Contrast {from=MaterialExpressionScalarParameter_3}
   ]
   properties {materialFunction=&CheapContrast}
  }
  MaterialExpressionMaterialFunctionCall_1 {
   class=MaterialExpressionMaterialFunctionCall
   inputs [Emissive {from=MaterialExpressionMaterialFunctionCall_2, output=Result}]
   properties {materialFunction=&MF_FX_EmissiveAdjustments}
  }
  MaterialExpressionMaterialFunctionCall_10 {
   class=MaterialExpressionMaterialFunctionCall
   inputs [
    Tile U {from=MaterialExpressionScalarParameter_23}
    Tile V {from=MaterialExpressionScalarParameter_27}
    UVSet {from=MaterialExpressionScalarParameter_28}
   ]
   properties {materialFunction=/SparksMaps/Environment/Shared/Materials/Functions/MF_Choose_UV_Set}
  }
  MaterialExpressionMaterialFunctionCall_11 {
   class=MaterialExpressionMaterialFunctionCall
   inputs [
    R {from=MaterialExpressionVectorParameter_1, output=Body Low End}
    G {from=MaterialExpressionVectorParameter_1, output=Body Low End}
    B {from=MaterialExpressionVectorParameter_1, output=Body Low End}
    A {from=MaterialExpressionVectorParameter_1, output=Body Low End}
   ]
   properties {materialFunction=&AppendMany}
  }
  MaterialExpressionMaterialFunctionCall_12 {
   class=MaterialExpressionMaterialFunctionCall
   inputs [
    BaseNormal {from=MaterialExpressionTextureSampleParameter2D_5, output=RGB}
    AdditionalNormal {from=MaterialExpressionTextureSampleParameter2D_2, output=RGB}
   ]
   properties {materialFunction=&BlendAngleCorrectedNormals}
  }
  MaterialExpressionMaterialFunctionCall_13 {
   class=MaterialExpressionMaterialFunctionCall
   inputs [
    Tile U {from=MaterialExpressionScalarParameter_29}
    Tile V {from=MaterialExpressionScalarParameter_30}
    UVSet {from=MaterialExpressionScalarParameter_47}
   ]
   properties {materialFunction=&MF_Choose_UV_Set}
  }
  MaterialExpressionMaterialFunctionCall_14 {
   class=MaterialExpressionMaterialFunctionCall
   inputs []
   properties {materialFunction=/SparksMaps/Environment/Shared/Materials/Functions/MF_Sparks_WPO}
  }
  MaterialExpressionMaterialFunctionCall_15 {
   class=MaterialExpressionMaterialFunctionCall
   inputs [
    TextureObject {from=MaterialExpressionTextureObjectParameter_3}
    TextureSize {from=""}
    WorldSpace {from=""}
    Normal {from=""}
    ProjectionTransitionContrast {from=""}
    WorldPosition {from=""}
    Use High Quality Normals {from=""}
   ]
   properties {materialFunction=&WorldAlignedNormal}
  }
  MaterialExpressionMaterialFunctionCall_16 {
   class=MaterialExpressionMaterialFunctionCall
   inputs [
    TextureObject {from=MaterialExpressionTextureObjectParameter_1}
    TextureSize {from=""}
    WorldPosition {from=""}
    Export Float 4 {from=MaterialExpressionStaticBool_0}
    World Space Normal {from=""}
    ProjectionTransitionContrast {from=""}
   ]
   properties {materialFunction=&WorldAlignedTexture}
  }
  MaterialExpressionMaterialFunctionCall_17 {
   class=MaterialExpressionMaterialFunctionCall
   inputs [
    TextureObject {from=MaterialExpressionTextureObjectParameter_2}
    TextureSize {from=""}
    WorldPosition {from=""}
    Export Float 4 {from=""}
    World Space Normal {from=""}
    ProjectionTransitionContrast {from=""}
   ]
   properties {materialFunction=&WorldAlignedTexture}
  }
  MaterialExpressionMaterialFunctionCall_18 {
   class=MaterialExpressionMaterialFunctionCall
   inputs [Float3 {from=MaterialExpressionStaticSwitchParameter_4}]
   properties {materialFunction=&BreakOutFloat3Components}
  }
  MaterialExpressionMaterialFunctionCall_19 {
   class=MaterialExpressionMaterialFunctionCall
   inputs [
    TextureObject {from=MaterialExpressionTextureObjectParameter_4}
    TextureSize {from=""}
    WorldSpace {from=""}
    Normal {from=""}
    ProjectionTransitionContrast {from=""}
    WorldPosition {from=""}
    Use High Quality Normals {from=""}
   ]
   properties {materialFunction=&WorldAlignedNormal}
  }
  MaterialExpressionMaterialFunctionCall_2 {
   class=MaterialExpressionMaterialFunctionCall
   inputs [
    Emissive {from=MaterialExpressionStaticSwitch_2}
    BeatPulse Strength {from=""}
    BeatPulse Min {from=""}
    BeatPulse Max {from=""}
   ]
   properties {materialFunction=/SparksMaps/Environment/MaterialFunctions/MF_MojoBeatPulse}
  }
  MaterialExpressionMaterialFunctionCall_20 {
   class=MaterialExpressionMaterialFunctionCall
   inputs [
    TextureObject {from=MaterialExpressionTextureObjectParameter_6}
    TextureSize {from=""}
    WorldPosition {from=""}
    Export Float 4 {from=MaterialExpressionStaticBool_1}
    World Space Normal {from=""}
    ProjectionTransitionContrast {from=""}
   ]
   properties {materialFunction=&WorldAlignedTexture}
  }
  MaterialExpressionMaterialFunctionCall_21 {
   class=MaterialExpressionMaterialFunctionCall
   inputs [
    TextureObject {from=MaterialExpressionTextureObjectParameter_5}
    TextureSize {from=""}
    WorldPosition {from=""}
    Export Float 4 {from=""}
    World Space Normal {from=""}
    ProjectionTransitionContrast {from=""}
   ]
   properties {materialFunction=&WorldAlignedTexture}
  }
  MaterialExpressionMaterialFunctionCall_22 {
   class=MaterialExpressionMaterialFunctionCall
   inputs [
    "Edge Intensity (Inverse)" {from=MaterialExpressionScalarParameter_38}
    Top Layer {from=MaterialExpressionNamedRerouteUsage_40}
    Exclusion Mask {from=MaterialExpressionChannelMaskParameter_4, output=R}
    Mask Channel {from=MaterialExpressionChannelMaskParameter_5, output=R}
    Vertex Color Channel {from=MaterialExpressionChannelMaskParameter_6, output=R}
    Cutoff {from=MaterialExpressionScalarParameter_32}
    Use Exclusion Mask {from=MaterialExpressionStaticBoolParameter_27}
    Use Vertex Color Mask {from=MaterialExpressionStaticBoolParameter_28}
    Slope Blend Sharpness {from=MaterialExpressionScalarParameter_36}
    Slope Blend Strength {from=MaterialExpressionScalarParameter_37}
    Use Slope Mask {from=MaterialExpressionStaticBoolParameter_29}
    Bottom Layer {from=MaterialExpressionNamedRerouteUsage_32}
    Opacity {from=MaterialExpressionScalarParameter_33}
    Overlay Blend {from=MaterialExpressionStaticBoolParameter_30}
    Hard Light Blend {from=MaterialExpressionStaticBoolParameter_31}
    Color Dodge Blend {from=MaterialExpressionStaticBoolParameter_32}
    Underlying Normal Intensity {from=MaterialExpressionScalarParameter_34}
    Fine Tune Edge {from=MaterialExpressionMaterialFunctionCall_23, output=RGBA}
    Edit Underlying Normal Intensity {from=MaterialExpressionStaticBoolParameter_34}
    Slope Intensity {from=MaterialExpressionScalarParameter_48}
    Vertex Falloff Mask Channel {from=MaterialExpressionChannelMaskParameter_7, output=R}
    Use Mask for Vertex Falloff {from=MaterialExpressionStaticBoolParameter_36}
    Falloff Noise Intensity {from=MaterialExpressionScalarParameter_53}
   ]
   properties {materialFunction=&/MF_Mask_Vert_Slope_Blend_A}
  }
  MaterialExpressionMaterialFunctionCall_23 {
   class=MaterialExpressionMaterialFunctionCall
   inputs [
    R {from=MaterialExpressionVectorParameter_4, output=Body Low End}
    G {from=MaterialExpressionVectorParameter_4, output=Body Low End}
    B {from=MaterialExpressionVectorParameter_4, output=Body Low End}
    A {from=MaterialExpressionVectorParameter_4, output=Body Low End}
   ]
   properties {materialFunction=&AppendMany}
  }
  MaterialExpressionMaterialFunctionCall_24 {
   class=MaterialExpressionMaterialFunctionCall
   inputs [Float3 {from=MaterialExpressionStaticSwitchParameter_7}]
   properties {materialFunction=&BreakOutFloat3Components}
  }
  MaterialExpressionMaterialFunctionCall_25 {
   class=MaterialExpressionMaterialFunctionCall
   inputs [
    Tile U {from=MaterialExpressionScalarParameter_41}
    Tile V {from=MaterialExpressionScalarParameter_42}
    UVSet {from=MaterialExpressionScalarParameter_43}
   ]
   properties {materialFunction=&MF_Choose_UV_Set}
  }
  MaterialExpressionMaterialFunctionCall_26 {
   class=MaterialExpressionMaterialFunctionCall
   inputs [
    Tile U {from=MaterialExpressionScalarParameter_14}
    Tile V {from=MaterialExpressionScalarParameter_15}
    UVSet {from=MaterialExpressionScalarParameter_54}
   ]
   properties {materialFunction=&MF_Choose_UV_Set}
  }
  MaterialExpressionMaterialFunctionCall_27 {
   class=MaterialExpressionMaterialFunctionCall
   inputs [
    "Edge Intensity (Inverse)" {from=MaterialExpressionScalarParameter_35}
    Top Layer {from=MaterialExpressionNamedRerouteUsage_39}
    Exclusion Mask {from=MaterialExpressionChannelMaskParameter_1, output=R}
    Mask Channel {from=MaterialExpressionChannelMaskParameter_3, output=R}
    Vertex Color Channel {from=MaterialExpressionChannelMaskParameter_2, output=R}
    Cutoff {from=MaterialExpressionScalarParameter_31}
    Use Exclusion Mask {from=MaterialExpressionStaticBoolParameter_21}
    Use Vertex Color Mask {from=MaterialExpressionStaticBoolParameter_22}
    Slope Blend Sharpness {from=MaterialExpressionScalarParameter_49}
    Slope Blend Strength {from=MaterialExpressionScalarParameter_50}
    Use Slope Mask {from=MaterialExpressionStaticBoolParameter_23}
    Bottom Layer {from=MaterialExpressionNamedRerouteUsage_38}
    Opacity {from=MaterialExpressionScalarParameter_39}
    Overlay Blend {from=MaterialExpressionStaticBoolParameter_24}
    Hard Light Blend {from=MaterialExpressionStaticBoolParameter_25}
    Color Dodge Blend {from=MaterialExpressionStaticBoolParameter_26}
    Underlying Normal Intensity {from=MaterialExpressionScalarParameter_40}
    Fine Tune Edge {from=MaterialExpressionMaterialFunctionCall_11, output=RGBA}
    Edit Underlying Normal Intensity {from=MaterialExpressionStaticBoolParameter_33}
    Slope Intensity {from=MaterialExpressionScalarParameter_51}
    Vertex Falloff Mask Channel {from=MaterialExpressionChannelMaskParameter_0, output=R}
    Use Mask for Vertex Falloff {from=MaterialExpressionStaticBoolParameter_35}
    Falloff Noise Intensity {from=MaterialExpressionScalarParameter_52}
   ]
   properties {materialFunction=&MF_Mask_Vert_Slope_Blend_A}
  }
  MaterialExpressionMaterialFunctionCall_28 {
   class=MaterialExpressionMaterialFunctionCall
   inputs [
    Hue Shift Percentage {from=MaterialExpressionScalarParameter_67}
    Texture {from=MaterialExpressionTextureSampleParameter2D_9, output=RGB}
   ]
   properties {materialFunction=&HueShift}
  }
  MaterialExpressionMaterialFunctionCall_29 {
   class=MaterialExpressionMaterialFunctionCall
   inputs [
    Tile U {from=MaterialExpressionScalarParameter_44}
    Tile V {from=MaterialExpressionScalarParameter_45}
    UVSet {from=MaterialExpressionScalarParameter_46}
   ]
   properties {materialFunction=&MF_Choose_UV_Set}
  }
  MaterialExpressionMaterialFunctionCall_3 {
   class=MaterialExpressionMaterialFunctionCall
   inputs [
    In {from=MaterialExpressionFresnel_1}
    Contrast {from=MaterialExpressionScalarParameter_9}
   ]
   properties {materialFunction=&CheapContrast}
  }
  MaterialExpressionMaterialFunctionCall_30 {
   class=MaterialExpressionMaterialFunctionCall
   inputs [
    In {from=MaterialExpressionTextureSampleParameter2D_4, output=Roughness}
    Contrast {from=MaterialExpressionScalarParameter_61}
   ]
   properties {materialFunction=&CheapContrast}
  }
  MaterialExpressionMaterialFunctionCall_31 {
   class=MaterialExpressionMaterialFunctionCall
   inputs [
    Normal {from=MaterialExpressionSaturate_1}
    CameraVector {from=MaterialExpressionCameraVectorWS_0}
   ]
   properties {materialFunction=/Engine/Functions/Engine_MaterialFunctions02/Utility/CustomReflectionVector}
  }
  MaterialExpressionMaterialFunctionCall_32 {
   class=MaterialExpressionMaterialFunctionCall
   inputs [
    In {from=MaterialExpressionTextureSampleParameter2D_4, output=Metallic}
    Contrast {from=MaterialExpressionScalarParameter_62}
   ]
   properties {materialFunction=&CheapContrast}
  }
  MaterialExpressionMaterialFunctionCall_33 {
   class=MaterialExpressionMaterialFunctionCall
   inputs [
    Hue Shift Percentage {from=MaterialExpressionScalarParameter_69}
    Texture {from=MaterialExpressionStaticSwitch_27}
   ]
   properties {materialFunction=&HueShift}
  }
  MaterialExpressionMaterialFunctionCall_37 {
   class=MaterialExpressionMaterialFunctionCall
   inputs [UV Scale {from=MaterialExpressionScalarParameter_68}]
   properties {materialFunction=/SparksMaps/Environment/Shared/Materials/Functions/MF_ObjectStabalisedScreenSpaceUVs}
  }
  MaterialExpressionMaterialFunctionCall_4 {
   class=MaterialExpressionMaterialFunctionCall
   inputs [
    Tile U {from=MaterialExpressionScalarParameter_0}
    Tile V {from=MaterialExpressionScalarParameter_5}
    UVSet {from=MaterialExpressionScalarParameter_24}
   ]
   properties {materialFunction=&MF_Choose_UV_Set}
  }
  MaterialExpressionMaterialFunctionCall_5 {
   class=MaterialExpressionMaterialFunctionCall
   inputs [
    Tile U {from=MaterialExpressionScalarParameter_21}
    Tile V {from=MaterialExpressionScalarParameter_22}
    UVSet {from=MaterialExpressionScalarParameter_20}
   ]
   properties {materialFunction=&MF_Choose_UV_Set}
  }
  MaterialExpressionMaterialFunctionCall_6 {
   class=MaterialExpressionMaterialFunctionCall
   inputs [
    Tile U {from=MaterialExpressionScalarParameter_11}
    Tile V {from=MaterialExpressionScalarParameter_12}
    UVSet {from=MaterialExpressionScalarParameter_26}
   ]
   properties {materialFunction=&MF_Choose_UV_Set}
  }
  MaterialExpressionMaterialFunctionCall_61 {
   class=MaterialExpressionMaterialFunctionCall
   inputs [Center Point World Position {from=MaterialExpressionObjectPositionWS_4}]
   properties {materialFunction=/Engine/Functions/Engine_MaterialFunctions02/UVs/WorldSpaceAlignedScreenCoordinates}
  }
  MaterialExpressionMaterialFunctionCall_64 {
   class=MaterialExpressionMaterialFunctionCall
   inputs [
    Tile U {from=MaterialExpressionScalarParameter_73}
    Tile V {from=MaterialExpressionScalarParameter_74}
    UVSet {from=MaterialExpressionScalarParameter_75}
   ]
   properties {materialFunction=&MF_Choose_UV_Set}
  }
  MaterialExpressionMaterialFunctionCall_7 {
   class=MaterialExpressionMaterialFunctionCall
   inputs [
    Tile U {from=MaterialExpressionScalarParameter_58}
    Tile V {from=MaterialExpressionScalarParameter_59}
    UVSet {from=MaterialExpressionScalarParameter_60}
   ]
   properties {materialFunction=&MF_Choose_UV_Set}
  }
  MaterialExpressionMaterialFunctionCall_8 {
   class=MaterialExpressionMaterialFunctionCall
   inputs [
    Normal {from=MaterialExpressionStaticSwitchParameter_0}
    Flatness {from=MaterialExpressionScalarParameter_6}
   ]
   properties {materialFunction=&FlattenNormal}
  }
  MaterialExpressionMaterialFunctionCall_9 {
   class=MaterialExpressionMaterialFunctionCall
   inputs [CustomUPosition {from=""}]
   properties {materialFunction=/SparksStageShow/Systems/StageManager/FXManager/MaterialFunctions/MF_MojoStageColors}
  }
  MaterialExpressionMultiply_0 {
   class=MaterialExpressionMultiply
   inputs [
    A {from=MaterialExpressionTextureCoordinate_0}
    B {from=MaterialExpressionScalarParameter_57}
   ]
   properties {constB=15}
  }
  MaterialExpressionMultiply_1 {
   class=MaterialExpressionMultiply
   inputs [
    A {from=MaterialExpressionReroute_3}
    B {from=MaterialExpressionStaticSwitch_3}
   ]
  }
  MaterialExpressionMultiply_10 {
   class=MaterialExpressionMultiply
   inputs [
    A {from=MaterialExpressionMultiply_7}
    B {from=MaterialExpressionMultiply_21}
   ]
  }
  MaterialExpressionMultiply_11 {
   class=MaterialExpressionMultiply
   inputs [
    A {from=MaterialExpressionMultiply_13}
    B {from=MaterialExpressionScalarParameter_56}
   ]
   properties {constB=1000}
  }
  MaterialExpressionMultiply_12 {
   class=MaterialExpressionMultiply
   inputs [
    A {from=MaterialExpressionNamedRerouteUsage_42}
    B {from=MaterialExpressionVectorParameter_5, output=RGB}
   ]
  }
  MaterialExpressionMultiply_13 {
   class=MaterialExpressionMultiply
   inputs [
    A {from=MaterialExpressionNamedRerouteUsage_13}
    B {from=MaterialExpressionPower_0}
   ]
  }
  MaterialExpressionMultiply_14 {
   class=MaterialExpressionMultiply
   inputs [
    A {from=MaterialExpressionMultiply_12}
    B {from=MaterialExpressionVectorParameter_5, output=A}
   ]
  }
  MaterialExpressionMultiply_15 {
   class=MaterialExpressionMultiply
   inputs [
    A {from=MaterialExpressionNamedRerouteUsage_43}
    B {from=MaterialExpressionVectorParameter_6, output=RGB}
   ]
  }
  MaterialExpressionMultiply_16 {
   class=MaterialExpressionMultiply
   inputs [
    A {from=MaterialExpressionMultiply_15}
    B {from=""}
   ]
  }
  MaterialExpressionMultiply_17 {
   class=MaterialExpressionMultiply
   inputs [
    A {from=MaterialExpressionReroute_2}
    B {from=MaterialExpressionDotProduct_0}
   ]
  }
  MaterialExpressionMultiply_18 {
   class=MaterialExpressionMultiply
   inputs [
    A {from=MaterialExpressionTextureSampleParameter2D_1, output=RGB}
    B {from=MaterialExpressionStaticSwitchParameter_1}
   ]
  }
  MaterialExpressionMultiply_19 {
   class=MaterialExpressionMultiply
   inputs [
    A {from=MaterialExpressionNamedRerouteUsage_69}
    B {from=MaterialExpressionVectorParameter_7, output=RGB}
   ]
  }
  MaterialExpressionMultiply_2 {
   class=MaterialExpressionMultiply
   inputs [
    A {from=MaterialExpressionNamedRerouteDeclaration_2}
    B {from=MaterialExpressionScalarParameter_17}
   ]
  }
  MaterialExpressionMultiply_21 {
   class=MaterialExpressionMultiply
   inputs [
    A {from=MaterialExpressionNamedRerouteUsage_41}
    B {from=MaterialExpressionScalarParameter_65}
   ]
  }
  MaterialExpressionMultiply_22 {
   class=MaterialExpressionMultiply
   inputs [
    A {from=MaterialExpressionStep_0}
    B {from=MaterialExpressionComponentMask_0}
   ]
  }
  MaterialExpressionMultiply_23 {
   class=MaterialExpressionMultiply
   inputs [
    A {from=MaterialExpressionNamedRerouteUsage_70}
    B {from=MaterialExpressionVectorParameter_8, output=RGB}
   ]
  }
  MaterialExpressionMultiply_24 {
   class=MaterialExpressionMultiply
   inputs [
    A {from=MaterialExpressionAdd_13}
    B {from=MaterialExpressionSaturate_5}
   ]
  }
  MaterialExpressionMultiply_25 {
   class=MaterialExpressionMultiply
   inputs [
    A {from=MaterialExpressionStep_1}
    B {from=MaterialExpressionComponentMask_3}
   ]
  }
  MaterialExpressionMultiply_26 {
   class=MaterialExpressionMultiply
   inputs [
    A {from=MaterialExpressionStep_2}
    B {from=MaterialExpressionComponentMask_4}
   ]
  }
  MaterialExpressionMultiply_27 {
   class=MaterialExpressionMultiply
   inputs [
    A {from=MaterialExpressionNamedRerouteUsage_71}
    B {from=MaterialExpressionVectorParameter_9, output=RGB}
   ]
  }
  MaterialExpressionMultiply_28 {
   class=MaterialExpressionMultiply
   inputs [
    A {from=MaterialExpressionMultiply_32}
    B {from=MaterialExpressionScalarParameter_80}
   ]
  }
  MaterialExpressionMultiply_29 {
   class=MaterialExpressionMultiply
   inputs [
    A {from=MaterialExpressionScalarParameter_82}
    B {from=MaterialExpressionSwitch_2}
   ]
  }
  MaterialExpressionMultiply_3 {
   class=MaterialExpressionMultiply
   inputs [
    A {from=MaterialExpressionMultiply_4}
    B {from=MaterialExpressionVectorParameter_2, output=A}
   ]
  }
  MaterialExpressionMultiply_31 {
   class=MaterialExpressionMultiply
   inputs [
    A {from=MaterialExpressionTextureSampleParameter2D_9, output=A}
    B {from=MaterialExpressionScalarParameter_78}
   ]
  }
  MaterialExpressionMultiply_32 {
   class=MaterialExpressionMultiply
   inputs [
    A {from=MaterialExpressionNamedRerouteUsage_74}
    B {from=MaterialExpressionNamedRerouteUsage_76}
   ]
  }
  MaterialExpressionMultiply_4 {
   class=MaterialExpressionMultiply
   inputs [
    A {from=MaterialExpressionVectorParameter_2, output=RGB}
    B {from=MaterialExpressionNamedRerouteUsage_17}
   ]
  }
  MaterialExpressionMultiply_49 {
   class=MaterialExpressionMultiply
   inputs [
    A {from=MaterialExpressionPower_5}
    B {from=MaterialExpressionScalarParameter_72}
   ]
   properties {constB=100}
  }
  MaterialExpressionMultiply_5 {
   class=MaterialExpressionMultiply
   inputs [
    A {from=MaterialExpressionStaticSwitch_0}
    B {from=MaterialExpressionNamedRerouteDeclaration_13}
   ]
  }
  MaterialExpressionMultiply_50 {
   class=MaterialExpressionMultiply
   inputs [
    A {from=MaterialExpressionTextureCoordinate_9}
    B {from=MaterialExpressionNamedRerouteDeclaration_45}
   ]
  }
  MaterialExpressionMultiply_51 {
   class=MaterialExpressionMultiply
   inputs [
    A {from=MaterialExpressionTextureSample_17, output=R}
    B {from=MaterialExpressionTextureSample_16, output=R}
   ]
  }
  MaterialExpressionMultiply_52 {
   class=MaterialExpressionMultiply
   inputs [
    A {from=MaterialExpressionNamedRerouteDeclaration_8}
    B {from=MaterialExpressionStaticSwitchParameter_18}
   ]
  }
  MaterialExpressionMultiply_55 {
   class=MaterialExpressionMultiply
   inputs [
    A {from=MaterialExpressionMaterialFunctionCall_61, output=Offset and Scaled UVS}
    B {from=MaterialExpressionMultiply_56}
   ]
  }
  MaterialExpressionMultiply_56 {
   class=MaterialExpressionMultiply
   inputs [
    A {from=MaterialExpressionNamedRerouteUsage_55}
    B {from=""}
   ]
  }
  MaterialExpressionMultiply_58 {
   class=MaterialExpressionMultiply
   inputs [
    A {from=MaterialExpressionMultiply_51}
    B {from=MaterialExpressionChannelMaskParameter_8, output=R}
   ]
  }
  MaterialExpressionMultiply_6 {
   class=MaterialExpressionMultiply
   inputs [
    A {from=MaterialExpressionMultiply_1}
    B {from=MaterialExpressionNamedRerouteUsage_21}
   ]
  }
  MaterialExpressionMultiply_7 {
   class=MaterialExpressionMultiply
   inputs [
    A {from=MaterialExpressionMultiply_8}
    B {from=MaterialExpressionNamedRerouteUsage_20}
   ]
  }
  MaterialExpressionMultiply_8 {
   class=MaterialExpressionMultiply
   inputs [
    A {from=MaterialExpressionScalarParameter_7}
    B {from=MaterialExpressionNamedRerouteUsage_19}
   ]
  }
  MaterialExpressionMultiply_9 {
   class=MaterialExpressionMultiply
   inputs [
    A {from=MaterialExpressionVectorParameter_3, output=RGB}
    B {from=MaterialExpressionVectorParameter_3, output=RGB}
   ]
  }
  MaterialExpressionNamedRerouteDeclaration_0 {
   class=MaterialExpressionNamedRerouteDeclaration
   inputs [None {from=MaterialExpressionMaterialFunctionCall_26, output=Result}]
   properties {name=GlobalUVControl, nodeColor=<rgba 0.9688 1 0 1>}
  }
  MaterialExpressionNamedRerouteDeclaration_1 {
   class=MaterialExpressionNamedRerouteDeclaration
   inputs [None {from=MaterialExpressionStaticSwitch_26}]
   properties {name=Final Diffuse, nodeColor=<rgba 0.1016 1 0 1>}
  }
  MaterialExpressionNamedRerouteDeclaration_10 {
   class=MaterialExpressionNamedRerouteDeclaration
   inputs [None {from=MaterialExpressionTextureSample_6, output=RGB}]
   properties {name=Mask RGB, nodeColor=<rgba 0.1953 1 0 1>}
  }
  MaterialExpressionNamedRerouteDeclaration_11 {
   class=MaterialExpressionNamedRerouteDeclaration
   inputs [None {from=MaterialExpressionStaticSwitch_15}]
   properties {name=Detail Diffuse, nodeColor=<rgba 1 0.0469 0 1>}
  }
  MaterialExpressionNamedRerouteDeclaration_12 {
   class=MaterialExpressionNamedRerouteDeclaration
   inputs [None {from=MaterialExpressionTextureSampleParameter2D_6, output=RGB}]
   properties {name=Diffuse RGB, nodeColor=<rgba 0 0.2734 1 1>}
  }
  MaterialExpressionNamedRerouteDeclaration_13 {
   class=MaterialExpressionNamedRerouteDeclaration
   inputs [None {from=MaterialExpressionTextureSampleParameter2D_7, output=RGB}]
   properties {name=Emissive RGB, nodeColor=<rgba 0 0.4375 1 1>}
  }
  MaterialExpressionNamedRerouteDeclaration_14 {
   class=MaterialExpressionNamedRerouteDeclaration
   inputs [None {from=MaterialExpressionStaticSwitch_5}]
   properties {name=Fresnel, nodeColor=<rgba 0.7578 1 0 1>}
  }
  MaterialExpressionNamedRerouteDeclaration_15 {
   class=MaterialExpressionNamedRerouteDeclaration
   inputs [None {from=MaterialExpressionStaticSwitch_13}]
   properties {name=Base Roughness, nodeColor=<rgba 0.9219 0 1 1>}
  }
  MaterialExpressionNamedRerouteDeclaration_16 {
   class=MaterialExpressionNamedRerouteDeclaration
   inputs [None {from=MaterialExpressionStaticSwitch_12}]
   properties {name=Base Metallic, nodeColor=<rgba 0 1 0.6016 1>}
  }
  MaterialExpressionNamedRerouteDeclaration_17 {
   class=MaterialExpressionNamedRerouteDeclaration
   inputs [None {from=MaterialExpressionTextureSampleParameter2D_3, output=Roughness}]
   properties {name=Fresnel Roughness, nodeColor=<rgba 1 0 0.8438 1>}
  }
  MaterialExpressionNamedRerouteDeclaration_18 {
   class=MaterialExpressionNamedRerouteDeclaration
   inputs [None {from=MaterialExpressionTextureSampleParameter2D_3, output=Metallic}]
   properties {name=Fresnel Metallic, nodeColor=<rgba 0 1 0.8359 1>}
  }
  MaterialExpressionNamedRerouteDeclaration_19 {
   class=MaterialExpressionNamedRerouteDeclaration
   inputs [None {from=MaterialExpressionTextureObjectParameter_11}]
   properties {name=Glitter Noise Tex, nodeColor=<rgba 1 0 0.1406 1>}
  }
  MaterialExpressionNamedRerouteDeclaration_2 {
   class=MaterialExpressionNamedRerouteDeclaration
   inputs [None {from=MaterialExpressionStaticSwitch_14}]
   properties {name=Fresnel_diffuse, nodeColor=<rgba 0.3828 0 1 1>}
  }
  MaterialExpressionNamedRerouteDeclaration_20 {
   class=MaterialExpressionNamedRerouteDeclaration
   inputs [None {from=MaterialExpressionMaterialFunctionCall_25, output=Result}]
   properties {name=Layer A UVs, nodeColor=<rgba 0 0.6016 1 1>}
  }
  MaterialExpressionNamedRerouteDeclaration_21 {
   class=MaterialExpressionNamedRerouteDeclaration
   inputs [None {from=MaterialExpressionMaterialFunctionCall_29, output=Result}]
   properties {name=Layer B UVs, nodeColor=<rgba 0.9219 1 0 1>}
  }
  MaterialExpressionNamedRerouteDeclaration_22 {
   class=MaterialExpressionNamedRerouteDeclaration
   inputs [None {from=MaterialExpressionMaterialFunctionCall_27, output=Result}]
   properties {name=Blend AOutput, nodeColor=<rgba 1 0.1172 0 1>}
  }
  MaterialExpressionNamedRerouteDeclaration_23 {
   class=MaterialExpressionNamedRerouteDeclaration
   inputs [None {from=MaterialExpressionMakeMaterialAttributes_2}]
   properties {name=Layer B, nodeColor=<rgba 0.7578 0 1 1>}
  }
  MaterialExpressionNamedRerouteDeclaration_24 {
   class=MaterialExpressionNamedRerouteDeclaration
   inputs [None {from=MaterialExpressionMakeMaterialAttributes_1}]
   properties {name=Layer A, nodeColor=<rgba 0 1 0.4375 1>}
  }
  MaterialExpressionNamedRerouteDeclaration_25 {
   class=MaterialExpressionNamedRerouteDeclaration
   inputs [None {from=MaterialExpressionComponentMask_6}]
   properties {name=Layer A Diffuse RGB, nodeColor=<rgba 0.1016 0 1 1>}
  }
  MaterialExpressionNamedRerouteDeclaration_26 {
   class=MaterialExpressionNamedRerouteDeclaration
   inputs [None {from=MaterialExpressionMakeMaterialAttributes_0}]
   properties {name=Main Layer, nodeColor=<rgba 0 0.7188 1 1>}
  }
  MaterialExpressionNamedRerouteDeclaration_27 {
   class=MaterialExpressionNamedRerouteDeclaration
   inputs [None {from=MaterialExpressionComponentMask_8}]
   properties {name=Layer B Diffuse RGB, nodeColor=<rgba 0.2188 1 0 1>}
  }
  MaterialExpressionNamedRerouteDeclaration_28 {
   class=MaterialExpressionNamedRerouteDeclaration
   inputs [None {from=MaterialExpressionTextureSampleParameterCube_0, output=RGB}]
   properties {name=CubeMap, nodeColor=<rgba 1 0 0.6328 1>}
  }
  MaterialExpressionNamedRerouteDeclaration_29 {
   class=MaterialExpressionNamedRerouteDeclaration
   inputs [None {from=MaterialExpressionTextureObjectParameter_0}]
   properties {name=MaskTexture, nodeColor=<rgba 1 0.8672 0 1>}
  }
  MaterialExpressionNamedRerouteDeclaration_3 {
   class=MaterialExpressionNamedRerouteDeclaration
   inputs [None {from=MaterialExpressionStaticSwitch_18}]
   properties {name=Final Metallic, nodeColor=<rgba 1 0 0.4219 1>}
  }
  MaterialExpressionNamedRerouteDeclaration_30 {
   class=MaterialExpressionNamedRerouteDeclaration
   inputs [None {from=MaterialExpressionStaticSwitch_6}]
   properties {name=Base Diffuse, nodeColor=<rgba 1 0.9375 0 1>}
  }
  MaterialExpressionNamedRerouteDeclaration_31 {
   class=MaterialExpressionNamedRerouteDeclaration
   inputs [None {from=MaterialExpressionMaterialFunctionCall_37, output=Result}]
   properties {name=Halftone UVs, nodeColor=<rgba 1 0.8203 0 1>}
  }
  MaterialExpressionNamedRerouteDeclaration_32 {
   class=MaterialExpressionNamedRerouteDeclaration
   inputs [None {from=MaterialExpressionScalarParameter_76}]
   properties {name=Halftone Rotation Amount, nodeColor=<rgba 0.5 0 1 1>}
  }
  MaterialExpressionNamedRerouteDeclaration_33 {
   class=MaterialExpressionNamedRerouteDeclaration
   inputs [None {from=MaterialExpressionScalarParameter_77}]
   properties {name=Halftone Dot Size, nodeColor=<rgba 0 1 0.1797 1>}
  }
  MaterialExpressionNamedRerouteDeclaration_34 {
   class=MaterialExpressionNamedRerouteDeclaration
   inputs [None {from=MaterialExpressionTextureSample_8, output=RGB}]
   properties {name=Halftone Mask, nodeColor=<rgba 1 0 0.1406 1>}
  }
  MaterialExpressionNamedRerouteDeclaration_35 {
   class=MaterialExpressionNamedRerouteDeclaration
   inputs [None {from=MaterialExpressionMultiply_22}]
   properties {name=Halftone A, nodeColor=<rgba 0 0.4609 1 1>}
  }
  MaterialExpressionNamedRerouteDeclaration_36 {
   class=MaterialExpressionNamedRerouteDeclaration
   inputs [None {from=MaterialExpressionMultiply_25}]
   properties {name=Halftone B, nodeColor=<rgba 0.7812 1 0 1>}
  }
  MaterialExpressionNamedRerouteDeclaration_37 {
   class=MaterialExpressionNamedRerouteDeclaration
   inputs [None {from=MaterialExpressionMultiply_26}]
   properties {name=Halftone C, nodeColor=<rgba 0.8984 0 1 1>}
  }
  MaterialExpressionNamedRerouteDeclaration_38 {
   class=MaterialExpressionNamedRerouteDeclaration
   inputs [None {from=MaterialExpressionSaturate_0}]
   properties {name=Halftone Colors Combined, nodeColor=<rgba 0 1 0.5781 1>}
  }
  MaterialExpressionNamedRerouteDeclaration_39 {
   class=MaterialExpressionNamedRerouteDeclaration
   inputs [None {from=MaterialExpressionSaturate_4}]
   properties {name=Halftone Lerp Alpha, nodeColor=<rgba 1 0.2578 0 1>}
  }
  MaterialExpressionNamedRerouteDeclaration_4 {
   class=MaterialExpressionNamedRerouteDeclaration
   inputs [None {from=MaterialExpressionStaticSwitch_19}]
   properties {name=Final Roughness, nodeColor=<rgba 0 0.7422 1 1>}
  }
  MaterialExpressionNamedRerouteDeclaration_45 {
   class=MaterialExpressionNamedRerouteDeclaration
   inputs [None {from=MaterialExpressionScalarParameter_70}]
   properties {name=Glitter UV Scale, nodeColor=<rgba 0.5938 0 1 1>}
  }
  MaterialExpressionNamedRerouteDeclaration_48 {
   class=MaterialExpressionNamedRerouteDeclaration
   inputs [None {from=MaterialExpressionStaticSwitchParameter_12}]
   properties {name=Glitter, nodeColor=<rgba 1 0 0.9141 1>}
  }
  MaterialExpressionNamedRerouteDeclaration_5 {
   class=MaterialExpressionNamedRerouteDeclaration
   inputs [None {from=MaterialExpressionStaticSwitch_10}]
   properties {name=Final Specular, nodeColor=<rgba 1 0.9375 0 1>}
  }
  MaterialExpressionNamedRerouteDeclaration_6 {
   class=MaterialExpressionNamedRerouteDeclaration
   inputs [None {from=MaterialExpressionTextureSampleParameter2D_6, output=A}]
   properties {name=Diffuse Alpha, nodeColor=<rgba 0.6172 0 1 1>}
  }
  MaterialExpressionNamedRerouteDeclaration_7 {
   class=MaterialExpressionNamedRerouteDeclaration
   inputs [None {from=MaterialExpressionMaterialFunctionCall_8, output=Result}]
   properties {name=Final Normals, nodeColor=<rgba 0 1 0.2969 1>}
  }
  MaterialExpressionNamedRerouteDeclaration_8 {
   class=MaterialExpressionNamedRerouteDeclaration
   inputs [None {from=MaterialExpressionCurveAtlasRowParameter_4}]
   properties {name=Glitter Gradient, nodeColor=<rgba 0 1 0.1797 1>}
  }
  MaterialExpressionNamedRerouteDeclaration_9 {
   class=MaterialExpressionNamedRerouteDeclaration
   inputs [None {from=MaterialExpressionStaticSwitch_29}]
   properties {name=Final Emissive, nodeColor=<rgba 0 0.1094 1 1>}
  }
  MaterialExpressionNamedRerouteUsage_0 {class=MaterialExpressionNamedRerouteUsage}
  MaterialExpressionNamedRerouteUsage_1 {class=MaterialExpressionNamedRerouteUsage}
  MaterialExpressionNamedRerouteUsage_10 {class=MaterialExpressionNamedRerouteUsage}
  MaterialExpressionNamedRerouteUsage_11 {class=MaterialExpressionNamedRerouteUsage}
  MaterialExpressionNamedRerouteUsage_12 {class=MaterialExpressionNamedRerouteUsage}
  MaterialExpressionNamedRerouteUsage_13 {class=MaterialExpressionNamedRerouteUsage}
  MaterialExpressionNamedRerouteUsage_14 {class=MaterialExpressionNamedRerouteUsage}
  MaterialExpressionNamedRerouteUsage_15 {class=MaterialExpressionNamedRerouteUsage}
  MaterialExpressionNamedRerouteUsage_16 {class=MaterialExpressionNamedRerouteUsage}
  MaterialExpressionNamedRerouteUsage_17 {class=MaterialExpressionNamedRerouteUsage}
  MaterialExpressionNamedRerouteUsage_18 {class=MaterialExpressionNamedRerouteUsage}
  MaterialExpressionNamedRerouteUsage_19 {class=MaterialExpressionNamedRerouteUsage}
  MaterialExpressionNamedRerouteUsage_2 {class=MaterialExpressionNamedRerouteUsage}
  MaterialExpressionNamedRerouteUsage_20 {class=MaterialExpressionNamedRerouteUsage}
  MaterialExpressionNamedRerouteUsage_21 {class=MaterialExpressionNamedRerouteUsage}
  MaterialExpressionNamedRerouteUsage_22 {class=MaterialExpressionNamedRerouteUsage}
  MaterialExpressionNamedRerouteUsage_23 {class=MaterialExpressionNamedRerouteUsage}
  MaterialExpressionNamedRerouteUsage_24 {class=MaterialExpressionNamedRerouteUsage}
  MaterialExpressionNamedRerouteUsage_25 {class=MaterialExpressionNamedRerouteUsage}
  MaterialExpressionNamedRerouteUsage_26 {class=MaterialExpressionNamedRerouteUsage}
  MaterialExpressionNamedRerouteUsage_27 {class=MaterialExpressionNamedRerouteUsage}
  MaterialExpressionNamedRerouteUsage_28 {class=MaterialExpressionNamedRerouteUsage}
  MaterialExpressionNamedRerouteUsage_29 {class=MaterialExpressionNamedRerouteUsage}
  MaterialExpressionNamedRerouteUsage_3 {class=MaterialExpressionNamedRerouteUsage}
  MaterialExpressionNamedRerouteUsage_30 {class=MaterialExpressionNamedRerouteUsage}
  MaterialExpressionNamedRerouteUsage_31 {class=MaterialExpressionNamedRerouteUsage}
  MaterialExpressionNamedRerouteUsage_32 {class=MaterialExpressionNamedRerouteUsage}
  MaterialExpressionNamedRerouteUsage_33 {class=MaterialExpressionNamedRerouteUsage}
  MaterialExpressionNamedRerouteUsage_34 {class=MaterialExpressionNamedRerouteUsage}
  MaterialExpressionNamedRerouteUsage_35 {class=MaterialExpressionNamedRerouteUsage}
  MaterialExpressionNamedRerouteUsage_36 {class=MaterialExpressionNamedRerouteUsage}
  MaterialExpressionNamedRerouteUsage_37 {class=MaterialExpressionNamedRerouteUsage}
  MaterialExpressionNamedRerouteUsage_38 {class=MaterialExpressionNamedRerouteUsage}
  MaterialExpressionNamedRerouteUsage_39 {class=MaterialExpressionNamedRerouteUsage}
  MaterialExpressionNamedRerouteUsage_4 {class=MaterialExpressionNamedRerouteUsage}
  MaterialExpressionNamedRerouteUsage_40 {class=MaterialExpressionNamedRerouteUsage}
  MaterialExpressionNamedRerouteUsage_41 {class=MaterialExpressionNamedRerouteUsage}
  MaterialExpressionNamedRerouteUsage_42 {class=MaterialExpressionNamedRerouteUsage}
  MaterialExpressionNamedRerouteUsage_43 {class=MaterialExpressionNamedRerouteUsage}
  MaterialExpressionNamedRerouteUsage_44 {class=MaterialExpressionNamedRerouteUsage}
  MaterialExpressionNamedRerouteUsage_45 {class=MaterialExpressionNamedRerouteUsage}
  MaterialExpressionNamedRerouteUsage_46 {class=MaterialExpressionNamedRerouteUsage}
  MaterialExpressionNamedRerouteUsage_47 {class=MaterialExpressionNamedRerouteUsage}
  MaterialExpressionNamedRerouteUsage_48 {class=MaterialExpressionNamedRerouteUsage}
  MaterialExpressionNamedRerouteUsage_49 {class=MaterialExpressionNamedRerouteUsage}
  MaterialExpressionNamedRerouteUsage_5 {class=MaterialExpressionNamedRerouteUsage}
  MaterialExpressionNamedRerouteUsage_50 {class=MaterialExpressionNamedRerouteUsage}
  MaterialExpressionNamedRerouteUsage_51 {class=MaterialExpressionNamedRerouteUsage}
  MaterialExpressionNamedRerouteUsage_52 {class=MaterialExpressionNamedRerouteUsage}
  MaterialExpressionNamedRerouteUsage_53 {class=MaterialExpressionNamedRerouteUsage}
  MaterialExpressionNamedRerouteUsage_54 {class=MaterialExpressionNamedRerouteUsage}
  MaterialExpressionNamedRerouteUsage_55 {class=MaterialExpressionNamedRerouteUsage}
  MaterialExpressionNamedRerouteUsage_56 {class=MaterialExpressionNamedRerouteUsage}
  MaterialExpressionNamedRerouteUsage_57 {class=MaterialExpressionNamedRerouteUsage}
  MaterialExpressionNamedRerouteUsage_58 {class=MaterialExpressionNamedRerouteUsage}
  MaterialExpressionNamedRerouteUsage_59 {class=MaterialExpressionNamedRerouteUsage}
  MaterialExpressionNamedRerouteUsage_6 {class=MaterialExpressionNamedRerouteUsage}
  MaterialExpressionNamedRerouteUsage_60 {class=MaterialExpressionNamedRerouteUsage}
  MaterialExpressionNamedRerouteUsage_61 {class=MaterialExpressionNamedRerouteUsage}
  MaterialExpressionNamedRerouteUsage_62 {class=MaterialExpressionNamedRerouteUsage}
  MaterialExpressionNamedRerouteUsage_64 {class=MaterialExpressionNamedRerouteUsage}
  MaterialExpressionNamedRerouteUsage_65 {class=MaterialExpressionNamedRerouteUsage}
  MaterialExpressionNamedRerouteUsage_66 {class=MaterialExpressionNamedRerouteUsage}
  MaterialExpressionNamedRerouteUsage_67 {class=MaterialExpressionNamedRerouteUsage}
  MaterialExpressionNamedRerouteUsage_68 {class=MaterialExpressionNamedRerouteUsage}
  MaterialExpressionNamedRerouteUsage_69 {class=MaterialExpressionNamedRerouteUsage}
  MaterialExpressionNamedRerouteUsage_7 {class=MaterialExpressionNamedRerouteUsage}
  MaterialExpressionNamedRerouteUsage_70 {class=MaterialExpressionNamedRerouteUsage}
  MaterialExpressionNamedRerouteUsage_71 {class=MaterialExpressionNamedRerouteUsage}
  MaterialExpressionNamedRerouteUsage_72 {class=MaterialExpressionNamedRerouteUsage}
  MaterialExpressionNamedRerouteUsage_73 {class=MaterialExpressionNamedRerouteUsage}
  MaterialExpressionNamedRerouteUsage_74 {class=MaterialExpressionNamedRerouteUsage}
  MaterialExpressionNamedRerouteUsage_76 {class=MaterialExpressionNamedRerouteUsage}
  MaterialExpressionNamedRerouteUsage_8 {class=MaterialExpressionNamedRerouteUsage}
  MaterialExpressionNamedRerouteUsage_9 {class=MaterialExpressionNamedRerouteUsage}
  MaterialExpressionNormalize_2 {class=MaterialExpressionNormalize, inputs [VectorInput {from=MaterialExpressionSubtract_0, output=""}]}
  MaterialExpressionObjectPositionWS_4 {class=MaterialExpressionObjectPositionWS, inputs [], properties {}}
  MaterialExpressionOneMinus_0 {class=MaterialExpressionOneMinus, inputs [None {from=MaterialExpressionNormalize_2}]}
  MaterialExpressionOneMinus_2 {class=MaterialExpressionOneMinus, inputs [None {from=MaterialExpressionMaterialFunctionCall_3, output=Result}]}
  MaterialExpressionOneMinus_4 {class=MaterialExpressionOneMinus, inputs [None {from=MaterialExpressionMaterialFunctionCall_0, output=Result}]}
  MaterialExpressionPower_0 {
   class=MaterialExpressionPower
   inputs [
    Base {from=MaterialExpressionMultiply_17}
    Exp {from=MaterialExpressionScalarParameter_55}
   ]
   properties {constExponent=500}
  }
  MaterialExpressionPower_1 {
   class=MaterialExpressionPower
   inputs [
    Base {from=MaterialExpressionComponentMask_1}
    Exp {from=MaterialExpressionScalarParameter_1}
   ]
  }
  MaterialExpressionPower_4 {
   class=MaterialExpressionPower
   inputs [
    Base {from=MaterialExpressionAdd_17}
    Exp {from=""}
   ]
   properties {constExponent=50}
  }
  MaterialExpressionPower_5 {
   class=MaterialExpressionPower
   inputs [
    Base {from=MaterialExpressionMultiply_52}
    Exp {from=MaterialExpressionScalarParameter_71}
   ]
   properties {constExponent=100}
  }
  MaterialExpressionReflectionVectorWS_0 {class=MaterialExpressionReflectionVectorWS, inputs [CustomWorldNormal {from="", output=""}], properties {}}
  MaterialExpressionReroute_0 {class=MaterialExpressionReroute, inputs [None {from=MaterialExpressionStaticSwitch_0}]}
  MaterialExpressionReroute_1 {class=MaterialExpressionReroute, inputs [None {from=MaterialExpressionMultiply_1}]}
  MaterialExpressionReroute_10 {class=MaterialExpressionReroute, inputs [None {from=MaterialExpressionAdd_5}]}
  MaterialExpressionReroute_11 {class=MaterialExpressionReroute, inputs [None {from=MaterialExpressionAdd_13}]}
  MaterialExpressionReroute_2 {class=MaterialExpressionReroute, inputs [None {from=MaterialExpressionSaturate_2}]}
  MaterialExpressionReroute_3 {class=MaterialExpressionReroute, inputs [None {from=MaterialExpressionMultiply_5}]}
  MaterialExpressionReroute_4 {class=MaterialExpressionReroute, inputs [None {from=MaterialExpressionTextureSampleParameter2D_4, output=Specular}]}
  MaterialExpressionReroute_5 {class=MaterialExpressionReroute, inputs [None {from=MaterialExpressionTextureSampleParameter2D_4, output=Roughness}]}
  MaterialExpressionReroute_6 {class=MaterialExpressionReroute, inputs [None {from=MaterialExpressionTextureSampleParameter2D_4, output=Metallic}]}
  MaterialExpressionReroute_7 {class=MaterialExpressionReroute, inputs [None {from=MaterialExpressionMultiply_7}]}
  MaterialExpressionReroute_8 {class=MaterialExpressionReroute, inputs [None {from=MaterialExpressionStaticSwitch_24}]}
  MaterialExpressionReroute_9 {class=MaterialExpressionReroute, inputs [None {from=MaterialExpressionStaticSwitch_25}]}
  MaterialExpressionRotator_0 {
   class=MaterialExpressionRotator
   inputs [
    Coordinate {from=MaterialExpressionNamedRerouteUsage_48, output=""}
    Time {from=MaterialExpressionNamedRerouteUsage_52, output=""}
   ]
   properties {speed=1}
  }
  MaterialExpressionRotator_1 {
   class=MaterialExpressionRotator
   inputs [
    Coordinate {from=MaterialExpressionAdd_6, output=""}
    Time {from=MaterialExpressionAdd_8, output=""}
   ]
   properties {speed=1}
  }
  MaterialExpressionRotator_2 {
   class=MaterialExpressionRotator
   inputs [
    Coordinate {from=MaterialExpressionAdd_9, output=""}
    Time {from=MaterialExpressionAdd_10, output=""}
   ]
   properties {speed=1}
  }
  MaterialExpressionSaturate_0 {class=MaterialExpressionSaturate, inputs [None {from=MaterialExpressionAdd_11}]}
  MaterialExpressionSaturate_1 {class=MaterialExpressionSaturate, inputs [None {from=MaterialExpressionAdd_1}]}
  MaterialExpressionSaturate_2 {class=MaterialExpressionSaturate, inputs [None {from=MaterialExpressionAdd_3}]}
  MaterialExpressionSaturate_4 {class=MaterialExpressionSaturate, inputs [None {from=MaterialExpressionPower_4}]}
  MaterialExpressionSaturate_5 {class=MaterialExpressionSaturate, inputs [None {from=MaterialExpressionDotProduct_1}]}
  MaterialExpressionScalarParameter_0 {class=MaterialExpressionScalarParameter, properties {defaultValue=1, group=Blend _Mask, parameterName="UV Tile U (Layer B)"}}
  MaterialExpressionScalarParameter_1 {class=MaterialExpressionScalarParameter, properties {defaultValue=2, group=05 Emissive, parameterName=GradientPower}}
  MaterialExpressionScalarParameter_10 {
   class=MaterialExpressionScalarParameter
   properties {defaultValue=0.002, group=01 Diffuse_Fresnel, parameterName=Diffuse_Fresnel_Falloff, sortPriority=43}
  }
  MaterialExpressionScalarParameter_11 {class=MaterialExpressionScalarParameter, properties {defaultValue=1, group=01 Diffuse_Fresnel, parameterName=DiffuseFresnel_U}}
  MaterialExpressionScalarParameter_12 {class=MaterialExpressionScalarParameter, properties {defaultValue=1, group=01 Diffuse_Fresnel, parameterName=DiffuseFresnel_V}}
  MaterialExpressionScalarParameter_13 {class=MaterialExpressionScalarParameter, properties {defaultValue=0.5, group=01 _Textures, parameterName=Override Specular, sliderMax=1}}
  MaterialExpressionScalarParameter_14 {class=MaterialExpressionScalarParameter, properties {defaultValue=1, group=01 _Textures, parameterName=Global UV - Tile U}}
  MaterialExpressionScalarParameter_15 {class=MaterialExpressionScalarParameter, properties {defaultValue=1, group=01 _Textures, parameterName=Global UV - Tile V}}
  MaterialExpressionScalarParameter_16 {
   class=MaterialExpressionScalarParameter
   properties {group=01 Diffuse_Fresnel, parameterName=Diffuse_Fresnel_Metallic, sliderMax=1, sortPriority=29}
  }
  MaterialExpressionScalarParameter_17 {
   class=MaterialExpressionScalarParameter
   properties {defaultValue=1, group=01 Diffuse_Fresnel, parameterName=Diffuse_Fresnel_Opacity, sliderMax=1, sortPriority=41}
  }
  MaterialExpressionScalarParameter_18 {class=MaterialExpressionScalarParameter, properties {group=01 _Textures, parameterName=Override Metallic, sliderMax=1}}
  MaterialExpressionScalarParameter_19 {class=MaterialExpressionScalarParameter, properties {defaultValue=0.5, group=01 _Textures, parameterName=Override Roughness, sliderMax=1}}
  MaterialExpressionScalarParameter_2 {class=MaterialExpressionScalarParameter, properties {defaultValue=0.06, group=05 Emissive, parameterName=Fresnel_Strength, sortPriority=42}}
  MaterialExpressionScalarParameter_20 {class=MaterialExpressionScalarParameter, properties {group=03 Diffuse Multiply Pattern, parameterName=Diffuse_UV_Set}}
  MaterialExpressionScalarParameter_21 {class=MaterialExpressionScalarParameter, properties {defaultValue=1, group=03 Diffuse Multiply Pattern, parameterName=Diffuse_Tile_U}}
  MaterialExpressionScalarParameter_22 {class=MaterialExpressionScalarParameter, properties {defaultValue=1, group=03 Diffuse Multiply Pattern, parameterName=Diffuse_Tile_V}}
  MaterialExpressionScalarParameter_23 {class=MaterialExpressionScalarParameter, properties {defaultValue=1, group=01 _Textures, parameterName=Detail Normal U}}
  MaterialExpressionScalarParameter_24 {class=MaterialExpressionScalarParameter, properties {group=Blend _Mask, parameterName="Blend Mask UV Set (Layer B)"}}
  MaterialExpressionScalarParameter_25 {
   class=MaterialExpressionScalarParameter
   properties {defaultValue=0.5, group=01 Diffuse_Fresnel, parameterName=Diffuse_Fresnel_Roughness, sliderMax=1, sortPriority=28}
  }
  MaterialExpressionScalarParameter_26 {class=MaterialExpressionScalarParameter, properties {group=01 Diffuse_Fresnel, parameterName=Diffuse_Fresnel_UV_Set, sortPriority=11}}
  MaterialExpressionScalarParameter_27 {class=MaterialExpressionScalarParameter, properties {defaultValue=1, group=01 _Textures, parameterName=Detail Normal V}}
  MaterialExpressionScalarParameter_28 {class=MaterialExpressionScalarParameter, properties {group=01 _Textures, parameterName=Detail Normal UV Set}}
  MaterialExpressionScalarParameter_29 {class=MaterialExpressionScalarParameter, properties {defaultValue=1, group=Blend _Mask, parameterName=UV Tile U}}
  MaterialExpressionScalarParameter_3 {class=MaterialExpressionScalarParameter, properties {defaultValue=0.03, group=05 Emissive, parameterName=Fresnel_Contrast, sortPriority=44}}
  MaterialExpressionScalarParameter_30 {class=MaterialExpressionScalarParameter, properties {defaultValue=1, group=Blend _Mask, parameterName=UV Tile V}}
  MaterialExpressionScalarParameter_31 {class=MaterialExpressionScalarParameter, properties {defaultValue=1, group=Blend A _Mask, parameterName="Cutoff (Blend A)", sliderMax=1}}
  MaterialExpressionScalarParameter_32 {class=MaterialExpressionScalarParameter, properties {defaultValue=1, group=Blend B _Mask, parameterName="Cutoff (Blend B)", sliderMax=1}}
  MaterialExpressionScalarParameter_33 {
   class=MaterialExpressionScalarParameter
   properties {defaultValue=1, group=Blend B Blend Modes, parameterName="_Opacity (Blend B)", sliderMax=1}
  }
  MaterialExpressionScalarParameter_34 {
   class=MaterialExpressionScalarParameter
   properties {group=Blend B _Textures, parameterName="Flatten Underlying Normal (Blend B)", sliderMax=1}
  }
  MaterialExpressionScalarParameter_35 {
   class=MaterialExpressionScalarParameter
   properties {defaultValue=3, group=Blend A Fine Tune Edge, parameterName="Edge_Intensity (Inverse) (Blend A)"}
  }
  MaterialExpressionScalarParameter_36 {class=MaterialExpressionScalarParameter, properties {defaultValue=10, group=Blend B Slope Mask, parameterName="Blend Sharpness (Blend B)"}}
  MaterialExpressionScalarParameter_37 {class=MaterialExpressionScalarParameter, properties {group=Blend B Slope Mask, parameterName="Blend Strength (Blend B)"}}
  MaterialExpressionScalarParameter_38 {
   class=MaterialExpressionScalarParameter
   properties {defaultValue=3, group=Blend B Fine Tune Edge, parameterName="Edge_Intensity (Inverse) (Blend B)"}
  }
  MaterialExpressionScalarParameter_39 {
   class=MaterialExpressionScalarParameter
   properties {defaultValue=1, group=Blend A Blend Modes, parameterName="_Opacity (Blend A)", sliderMax=1}
  }
  MaterialExpressionScalarParameter_4 {class=MaterialExpressionScalarParameter, properties {defaultValue=50, group=05 Emissive, parameterName=Fresnel_Falloff, sortPriority=43}}
  MaterialExpressionScalarParameter_40 {
   class=MaterialExpressionScalarParameter
   properties {group=Blend A _Textures, parameterName="Flatten Underlying Normal (Blend A)", sliderMax=1}
  }
  MaterialExpressionScalarParameter_41 {class=MaterialExpressionScalarParameter, properties {defaultValue=1, group=Blend A _Textures, parameterName=U Layer A}}
  MaterialExpressionScalarParameter_42 {class=MaterialExpressionScalarParameter, properties {defaultValue=1, group=Blend A _Textures, parameterName=V Layer A}}
  MaterialExpressionScalarParameter_43 {class=MaterialExpressionScalarParameter, properties {group=Blend A _Textures, parameterName=Layer A UV Set}}
  MaterialExpressionScalarParameter_44 {class=MaterialExpressionScalarParameter, properties {group=Blend B _Textures, parameterName=U Layer B}}
  MaterialExpressionScalarParameter_45 {class=MaterialExpressionScalarParameter, properties {group=Blend B _Textures, parameterName=V Layer B}}
  MaterialExpressionScalarParameter_46 {class=MaterialExpressionScalarParameter, properties {group=Blend B _Textures, parameterName=Layer B UV Set}}
  MaterialExpressionScalarParameter_47 {class=MaterialExpressionScalarParameter, properties {group=Blend _Mask, parameterName=Blend Mask UV Set}}
  MaterialExpressionScalarParameter_48 {class=MaterialExpressionScalarParameter, properties {group=Blend B Slope Mask, parameterName="Slope Intensity (Blend B)", sliderMax=1}}
  MaterialExpressionScalarParameter_49 {class=MaterialExpressionScalarParameter, properties {group=Blend A Slope Mask, parameterName="Blend Sharpness (Blend A)"}}
  MaterialExpressionScalarParameter_5 {class=MaterialExpressionScalarParameter, properties {defaultValue=1, group=Blend _Mask, parameterName="UV Tile V (Layer B)"}}
  MaterialExpressionScalarParameter_50 {class=MaterialExpressionScalarParameter, properties {group=Blend A Slope Mask, parameterName="Blend Strength (Blend A)"}}
  MaterialExpressionScalarParameter_51 {class=MaterialExpressionScalarParameter, properties {group=Blend A Slope Mask, parameterName="Blend Intensity (Blend A)"}}
  MaterialExpressionScalarParameter_52 {class=MaterialExpressionScalarParameter, properties {group=Blend A _Mask, parameterName="Falloff Noise Intensity (Blend A)", sliderMax=1}}
  MaterialExpressionScalarParameter_53 {class=MaterialExpressionScalarParameter, properties {group=Blend B _Mask, parameterName="Fall Off Noise Intensity (Blend B)", sliderMax=1}}
  MaterialExpressionScalarParameter_54 {class=MaterialExpressionScalarParameter, properties {group=01 _Textures, parameterName=Global UV Set}}
  MaterialExpressionScalarParameter_55 {class=MaterialExpressionScalarParameter, properties {defaultValue=750, group=06 Glitter, parameterName=Extra Glitter Exponant}}
  MaterialExpressionScalarParameter_56 {class=MaterialExpressionScalarParameter, properties {defaultValue=1000, group=06 Glitter, parameterName=Extra Glitter Intensity}}
  MaterialExpressionScalarParameter_57 {class=MaterialExpressionScalarParameter, properties {defaultValue=10, group=06 Glitter, parameterName=Extra Glitter UV Scale}}
  MaterialExpressionScalarParameter_58 {class=MaterialExpressionScalarParameter, properties {defaultValue=1, group=05 Emissive, parameterName=Emissive Texture U}}
  MaterialExpressionScalarParameter_59 {class=MaterialExpressionScalarParameter, properties {defaultValue=1, group=05 Emissive, parameterName=Emissive Texture V}}
  MaterialExpressionScalarParameter_6 {class=MaterialExpressionScalarParameter, properties {group=01 _Textures, parameterName=_Flatten Normal, sliderMax=1}}
  MaterialExpressionScalarParameter_60 {class=MaterialExpressionScalarParameter, properties {group=05 Emissive, parameterName=Emissive UV Set}}
  MaterialExpressionScalarParameter_61 {class=MaterialExpressionScalarParameter, properties {group=01 _Textures, parameterName=Custom Roughness Contrast, sliderMax=1, sliderMin=-1}}
  MaterialExpressionScalarParameter_62 {class=MaterialExpressionScalarParameter, properties {group=01 _Textures, parameterName=Custom Metallic Contrast, sliderMax=1, sliderMin=-1}}
  MaterialExpressionScalarParameter_63 {
   class=MaterialExpressionScalarParameter
   properties {desc="0 = Primary, 1 = Secondary, 2 = Tertiary", group=05 Emissive, parameterName=Choose Stage Color, sliderMax=2}
  }
  MaterialExpressionScalarParameter_64 {class=MaterialExpressionScalarParameter, properties {defaultValue=0.5, group=05 Emissive, parameterName=DiffuseCubeMapBlend}}
  MaterialExpressionScalarParameter_65 {class=MaterialExpressionScalarParameter, properties {defaultValue=1, group=05 Emissive, parameterName=EmissiveCubeMapIntensity}}
  MaterialExpressionScalarParameter_66 {class=MaterialExpressionScalarParameter, properties {group=07 Graffiti, parameterName=Graffiti UV Set, sortPriority=2}}
  MaterialExpressionScalarParameter_67 {
   class=MaterialExpressionScalarParameter
   properties {bUseCustomPrimitiveData=true, group=07 Graffiti, parameterName=Graffiti Hue Shift, sliderMax=1, sortPriority=4}
  }
  MaterialExpressionScalarParameter_68 {class=MaterialExpressionScalarParameter, properties {defaultValue=0.5, group=08 Halftone, parameterName=Halftone Dot Scale, sortPriority=1}}
  MaterialExpressionScalarParameter_69 {class=MaterialExpressionScalarParameter, properties {group=01 _Textures, parameterName=Final Diffuse Hue Shift, sortPriority=34}}
  MaterialExpressionScalarParameter_7 {
   class=MaterialExpressionScalarParameter
   properties {defaultValue=1, group=05 Emissive, parameterName=Emissive_Albedo_Intensity, sliderMax=20}
  }
  MaterialExpressionScalarParameter_70 {class=MaterialExpressionScalarParameter, properties {defaultValue=1, group=06 Glitter, parameterName=__Glitter UV Scale}}
  MaterialExpressionScalarParameter_71 {class=MaterialExpressionScalarParameter, properties {defaultValue=5, group=06 Glitter, parameterName=_Glitter Exponent}}
  MaterialExpressionScalarParameter_72 {class=MaterialExpressionScalarParameter, properties {defaultValue=2500, group=06 Glitter, parameterName=_Glitter Intensity}}
  MaterialExpressionScalarParameter_73 {class=MaterialExpressionScalarParameter, properties {defaultValue=1, group=06 Glitter, parameterName=Glitter Mask U}}
  MaterialExpressionScalarParameter_74 {class=MaterialExpressionScalarParameter, properties {defaultValue=1, group=06 Glitter, parameterName=Glitter Mask V}}
  MaterialExpressionScalarParameter_75 {class=MaterialExpressionScalarParameter, properties {group=06 Glitter, parameterName=Glitter Mask UV Set}}
  MaterialExpressionScalarParameter_76 {
   class=MaterialExpressionScalarParameter
   properties {defaultValue=0.8, group=08 Halftone, parameterName=Halftone Rotation Amount, sliderMax=1.5, sortPriority=2}
  }
  MaterialExpressionScalarParameter_77 {
   class=MaterialExpressionScalarParameter
   properties {defaultValue=0.25, group=08 Halftone, parameterName=Halftone Dot Size, sliderMax=0.5, sortPriority=3}
  }
  MaterialExpressionScalarParameter_78 {class=MaterialExpressionScalarParameter, properties {defaultValue=1, group=07 Graffiti, parameterName=Grafitti Opacity, sliderMax=1}}
  MaterialExpressionScalarParameter_79 {class=MaterialExpressionScalarParameter, properties {group=07 Graffiti, parameterName=Desaturation Fraction}}
  MaterialExpressionScalarParameter_8 {
   class=MaterialExpressionScalarParameter
   properties {defaultValue=3, group=01 Diffuse_Fresnel, parameterName=Diffuse_Fresnel_Strength, sortPriority=42}
  }
  MaterialExpressionScalarParameter_80 {
   class=MaterialExpressionScalarParameter
   properties {defaultValue=1, group=08 Halftone, parameterName=Halftone Brightness, sliderMax=100, sliderMin=1, sortPriority=8}
  }
  MaterialExpressionScalarParameter_81 {class=MaterialExpressionScalarParameter, properties {parameterName=Halftone Mask UV Set}}
  MaterialExpressionScalarParameter_82 {class=MaterialExpressionScalarParameter, properties {defaultValue=1, parameterName=Halftone Mask UV Scale}}
  MaterialExpressionScalarParameter_9 {
   class=MaterialExpressionScalarParameter
   properties {defaultValue=0.03, group=01 Diffuse_Fresnel, parameterName=Disffuse_Fresnel_Contrast, sortPriority=44}
  }
  MaterialExpressionStaticBoolParameter_0 {class=MaterialExpressionStaticBoolParameter, inputs [], properties {group=05 Emissive, parameterName=UseStagePalette}}
  MaterialExpressionStaticBoolParameter_1 {class=MaterialExpressionStaticBoolParameter, inputs [], properties {group=05 Emissive, parameterName=UseAlbedoEmissive, sortPriority=33}}
  MaterialExpressionStaticBoolParameter_10 {class=MaterialExpressionStaticBoolParameter, inputs [], properties {group=06 Glitter, parameterName=___Use Glitter}}
  MaterialExpressionStaticBoolParameter_11 {class=MaterialExpressionStaticBoolParameter, inputs [], properties {group=01 _Textures, parameterName=Override Specular, sortPriority=24}}
  MaterialExpressionStaticBoolParameter_12 {class=MaterialExpressionStaticBoolParameter, inputs [], properties {group=01 _Textures, parameterName=Override Metallic, sortPriority=24}}
  MaterialExpressionStaticBoolParameter_13 {class=MaterialExpressionStaticBoolParameter, inputs [], properties {group=01 _Textures, parameterName=Override Roughness, sortPriority=24}}
  MaterialExpressionStaticBoolParameter_14 {
   class=MaterialExpressionStaticBoolParameter
   inputs []
   properties {group=01 Diffuse_Fresnel, parameterName=Inner_Diffuse_Fresnel, sortPriority=40}
  }
  MaterialExpressionStaticBoolParameter_15 {
   class=MaterialExpressionStaticBoolParameter
   inputs []
   properties {group=03 Diffuse Multiply Pattern, parameterName=Choose_UV_Diffuse_Multiply, sortPriority=20}
  }
  MaterialExpressionStaticBoolParameter_16 {
   class=MaterialExpressionStaticBoolParameter
   inputs []
   properties {group=01 Diffuse_Fresnel, parameterName=Overide Fresnel roughness with Texture, sortPriority=25}
  }
  MaterialExpressionStaticBoolParameter_17 {
   class=MaterialExpressionStaticBoolParameter
   inputs []
   properties {group=01 Diffuse_Fresnel, parameterName=Overide Fresnel metallic with Texture, sortPriority=24}
  }
  MaterialExpressionStaticBoolParameter_18 {class=MaterialExpressionStaticBoolParameter, inputs [], properties {group=01 _Textures, parameterName=Custom Roughness Contrast}}
  MaterialExpressionStaticBoolParameter_19 {
   class=MaterialExpressionStaticBoolParameter
   inputs []
   properties {group=01 Diffuse_Fresnel, parameterName=Use_Diffuse_Fresnel, sortPriority=5}
  }
  MaterialExpressionStaticBoolParameter_2 {class=MaterialExpressionStaticBoolParameter, inputs [], properties {group=05 Emissive, parameterName=UseEmissive}}
  MaterialExpressionStaticBoolParameter_20 {
   class=MaterialExpressionStaticBoolParameter
   inputs []
   properties {group=01 Diffuse_Fresnel, parameterName=Use_Diffuse_Fresnel, sortPriority=5}
  }
  MaterialExpressionStaticBoolParameter_21 {class=MaterialExpressionStaticBoolParameter, inputs [], properties {group=Blend A _Mask, parameterName="_Use Exclusion Mask (Blend A)"}}
  MaterialExpressionStaticBoolParameter_22 {
   class=MaterialExpressionStaticBoolParameter
   inputs []
   properties {defaultValue=true, group=Blend A _Mask, parameterName="_Use Vertex Color Mask (Blend A)"}
  }
  MaterialExpressionStaticBoolParameter_23 {class=MaterialExpressionStaticBoolParameter, inputs [], properties {group=Blend A Slope Mask, parameterName="_Use Slope Mask  (Blend A)"}}
  MaterialExpressionStaticBoolParameter_24 {class=MaterialExpressionStaticBoolParameter, inputs [], properties {group=Blend A Blend Modes, parameterName="Overlay (Blend A)"}}
  MaterialExpressionStaticBoolParameter_25 {class=MaterialExpressionStaticBoolParameter, inputs [], properties {group=Blend A Blend Modes, parameterName="Hard Light (Blend A)"}}
  MaterialExpressionStaticBoolParameter_26 {class=MaterialExpressionStaticBoolParameter, inputs [], properties {group=Blend A Blend Modes, parameterName="Color Dodge (Blend A)"}}
  MaterialExpressionStaticBoolParameter_27 {class=MaterialExpressionStaticBoolParameter, inputs [], properties {group=Blend B _Mask, parameterName="_Use Exclusion Mask (Blend 2)"}}
  MaterialExpressionStaticBoolParameter_28 {
   class=MaterialExpressionStaticBoolParameter
   inputs []
   properties {defaultValue=true, group=Blend B _Mask, parameterName="_Use Vertex Color Mask (Blend 2)"}
  }
  MaterialExpressionStaticBoolParameter_29 {
   class=MaterialExpressionStaticBoolParameter
   inputs []
   properties {defaultValue=true, group=Blend B Slope Mask, parameterName="_Use Slope Mask  (Blend B)"}
  }
  MaterialExpressionStaticBoolParameter_3 {
   class=MaterialExpressionStaticBoolParameter
   inputs []
   properties {group=05 Emissive, parameterName=Use_UV_Gradient_for_Emiss_Mask, sortPriority=33}
  }
  MaterialExpressionStaticBoolParameter_30 {class=MaterialExpressionStaticBoolParameter, inputs [], properties {group=Blend B Blend Modes, parameterName="Overlay (Blend B)"}}
  MaterialExpressionStaticBoolParameter_31 {class=MaterialExpressionStaticBoolParameter, inputs [], properties {group=Blend B Blend Modes, parameterName="Hard Light (Blend B)"}}
  MaterialExpressionStaticBoolParameter_32 {class=MaterialExpressionStaticBoolParameter, inputs [], properties {group=Blend B Blend Modes, parameterName="Color Dodge (Blend B)"}}
  MaterialExpressionStaticBoolParameter_33 {
   class=MaterialExpressionStaticBoolParameter
   inputs []
   properties {group=Blend A _Textures, parameterName="Flatten Underlying Normal (Blend A)"}
  }
  MaterialExpressionStaticBoolParameter_34 {
   class=MaterialExpressionStaticBoolParameter
   inputs []
   properties {group=Blend B _Textures, parameterName="Flatten Underlying Normal Toggle (Blend B)"}
  }
  MaterialExpressionStaticBoolParameter_35 {
   class=MaterialExpressionStaticBoolParameter
   inputs []
   properties {group=Blend A _Mask, parameterName="_Use Mask for Vertex Falloff (Blend A)"}
  }
  MaterialExpressionStaticBoolParameter_36 {
   class=MaterialExpressionStaticBoolParameter
   inputs []
   properties {group=Blend B _Mask, parameterName="_Use Mask for Vertex Falloff (Blend B)"}
  }
  MaterialExpressionStaticBoolParameter_37 {class=MaterialExpressionStaticBoolParameter, inputs [], properties {group=01 _Textures, parameterName=Custom Metallic Contrast}}
  MaterialExpressionStaticBoolParameter_38 {class=MaterialExpressionStaticBoolParameter, inputs [], properties {group=05 Emissive, parameterName=UseCubeMapforEmissive}}
  MaterialExpressionStaticBoolParameter_39 {class=MaterialExpressionStaticBoolParameter, inputs [], properties {group=05 Emissive, parameterName=BlendCubeMapwithDiffuse}}
  MaterialExpressionStaticBoolParameter_4 {
   class=MaterialExpressionStaticBoolParameter
   inputs []
   properties {group=05 Emissive, parameterName=UseFresnel_with_Emissive, sortPriority=40}
  }
  MaterialExpressionStaticBoolParameter_40 {class=MaterialExpressionStaticBoolParameter, inputs [], properties {group=07 Graffiti, parameterName=Use Graffiti, sortPriority=0}}
  MaterialExpressionStaticBoolParameter_41 {class=MaterialExpressionStaticBoolParameter, inputs [], properties {group=07 Graffiti, parameterName=Use Color Variation, sortPriority=3}}
  MaterialExpressionStaticBoolParameter_42 {
   class=MaterialExpressionStaticBoolParameter
   inputs []
   properties {group=01 _Textures, parameterName=HueShift Final Diffuse, sortPriority=33}
  }
  MaterialExpressionStaticBoolParameter_43 {class=MaterialExpressionStaticBoolParameter, inputs [], properties {group=08 Halftone, parameterName=Use Halftone, sortPriority=0}}
  MaterialExpressionStaticBoolParameter_44 {class=MaterialExpressionStaticBoolParameter, inputs [], properties {group=07 Graffiti, parameterName=Use Desaturation}}
  MaterialExpressionStaticBoolParameter_45 {
   class=MaterialExpressionStaticBoolParameter
   inputs []
   properties {group=08 Halftone, parameterName=Use Shadow to Mask Halftone, sortPriority=4}
  }
  MaterialExpressionStaticBoolParameter_46 {
   class=MaterialExpressionStaticBoolParameter
   inputs []
   properties {group=08 Halftone, parameterName=Add Halftone to Emissive, sortPriority=1}
  }
  MaterialExpressionStaticBoolParameter_5 {class=MaterialExpressionStaticBoolParameter, inputs [], properties {group=05 Emissive, parameterName=Fresnel_Inner_Glow, sortPriority=41}}
  MaterialExpressionStaticBoolParameter_6 {
   class=MaterialExpressionStaticBoolParameter
   inputs []
   properties {group=01 Diffuse_Fresnel, parameterName=Use_Texture_Fresnel, sortPriority=8}
  }
  MaterialExpressionStaticBoolParameter_7 {
   class=MaterialExpressionStaticBoolParameter
   inputs []
   properties {group=01 Diffuse_Fresnel, parameterName=Use_Diffuse_Fresnel, sortPriority=5}
  }
  MaterialExpressionStaticBoolParameter_8 {class=MaterialExpressionStaticBoolParameter, inputs [], properties {group=01 _Textures, parameterName=Override Metallic, sortPriority=24}}
  MaterialExpressionStaticBoolParameter_9 {class=MaterialExpressionStaticBoolParameter, inputs [], properties {group=01 _Textures, parameterName=Override Roughness, sortPriority=24}}
  MaterialExpressionStaticBool_0 {class=MaterialExpressionStaticBool, properties {value=true}}
  MaterialExpressionStaticBool_1 {class=MaterialExpressionStaticBool, properties {value=true}}
  MaterialExpressionStaticSwitchParameter_0 {
   class=MaterialExpressionStaticSwitchParameter
   inputs [
    True {from=MaterialExpressionMaterialFunctionCall_12, output=Result}
    False {from=MaterialExpressionTextureSampleParameter2D_5, output=RGB}
   ]
   properties {group=01 _Textures, parameterName=_Use Detail Normal}
  }
  MaterialExpressionStaticSwitchParameter_1 {
   class=MaterialExpressionStaticSwitchParameter
   inputs [
    True {from=MaterialExpressionLinearInterpolate_2, output=""}
    False {from=MaterialExpressionNamedRerouteUsage_18, output=""}
   ]
   properties {group=01 _Textures, parameterName=Enable Tint, sortPriority=35}
  }
  MaterialExpressionStaticSwitchParameter_10 {
   class=MaterialExpressionStaticSwitchParameter
   inputs [
    True {from=MaterialExpressionLinearInterpolate_5, output=""}
    False {from=MaterialExpressionComponentMask_8, output=""}
   ]
   properties {group=Blend B _Textures, parameterName=Tint Layer B}
  }
  MaterialExpressionStaticSwitchParameter_11 {
   class=MaterialExpressionStaticSwitchParameter
   inputs [
    True {from=MaterialExpressionTextureSample_7, output=RGB}
    False {from=MaterialExpressionNamedRerouteUsage_30, output=""}
   ]
   properties {group=Blend _Mask, parameterName=Use a different UV set for layer B}
  }
  MaterialExpressionStaticSwitchParameter_12 {
   class=MaterialExpressionStaticSwitchParameter
   inputs [
    True {from=MaterialExpressionAdd_0, output=""}
    False {from=MaterialExpressionMultiply_49, output=""}
   ]
   properties {defaultValue=true, group=06 Glitter, parameterName=_Use Extra Glitter}
  }
  MaterialExpressionStaticSwitchParameter_17 {
   class=MaterialExpressionStaticSwitchParameter
   inputs [
    True {from=MaterialExpressionStaticSwitchParameter_2, output=""}
    False {from=MaterialExpressionNamedRerouteUsage_58, output=""}
   ]
   properties {defaultValue=true, group=Blend __Layer Toggle, parameterName=_Use Layers}
  }
  MaterialExpressionStaticSwitchParameter_18 {
   class=MaterialExpressionStaticSwitchParameter
   inputs [
    True {from=MaterialExpressionMultiply_58, output=""}
    False {from=MaterialExpressionMultiply_51, output=""}
   ]
   properties {group=06 Glitter, parameterName=_Use Glitter Mask}
  }
  MaterialExpressionStaticSwitchParameter_2 {
   class=MaterialExpressionStaticSwitchParameter
   inputs [
    True {from=MaterialExpressionMaterialFunctionCall_22, output=Result}
    False {from=MaterialExpressionNamedRerouteUsage_59, output=""}
   ]
   properties {group=Blend __Layer Toggle, parameterName=Use 2 Layers}
  }
  MaterialExpressionStaticSwitchParameter_3 {
   class=MaterialExpressionStaticSwitchParameter
   inputs [
    True {from=MaterialExpressionMaterialFunctionCall_16, output=XYZ Texture}
    False {from=MaterialExpressionTextureSample_0, output=RGBA}
   ]
   properties {group=Blend A _Textures, parameterName=Layer A Triplanar}
  }
  MaterialExpressionStaticSwitchParameter_4 {
   class=MaterialExpressionStaticSwitchParameter
   inputs [
    True {from=MaterialExpressionMaterialFunctionCall_17, output=XYZ Texture}
    False {from=MaterialExpressionTextureSample_1, output=RGB}
   ]
   properties {group=Blend A _Textures, parameterName=Layer A Triplanar}
  }
  MaterialExpressionStaticSwitchParameter_5 {
   class=MaterialExpressionStaticSwitchParameter
   inputs [
    True {from=MaterialExpressionMaterialFunctionCall_15, output=XYZ Texture}
    False {from=MaterialExpressionTextureSample_2, output=RGB}
   ]
   properties {group=Blend A _Textures, parameterName=Layer A Triplanar}
  }
  MaterialExpressionStaticSwitchParameter_6 {
   class=MaterialExpressionStaticSwitchParameter
   inputs [
    True {from=MaterialExpressionMaterialFunctionCall_20, output=XYZ Texture}
    False {from=MaterialExpressionTextureSample_3, output=RGBA}
   ]
   properties {group=Blend B _Textures, parameterName=Layer B Triplanar}
  }
  MaterialExpressionStaticSwitchParameter_7 {
   class=MaterialExpressionStaticSwitchParameter
   inputs [
    True {from=MaterialExpressionMaterialFunctionCall_21, output=XYZ Texture}
    False {from=MaterialExpressionTextureSample_4, output=RGB}
   ]
   properties {group=Blend B _Textures, parameterName=Layer B Triplanar}
  }
  MaterialExpressionStaticSwitchParameter_8 {
   class=MaterialExpressionStaticSwitchParameter
   inputs [
    True {from=MaterialExpressionMaterialFunctionCall_19, output=XYZ Texture}
    False {from=MaterialExpressionTextureSample_5, output=RGB}
   ]
   properties {group=Blend B _Textures, parameterName=Layer B Triplanar}
  }
  MaterialExpressionStaticSwitchParameter_9 {
   class=MaterialExpressionStaticSwitchParameter
   inputs [
    True {from=MaterialExpressionLinearInterpolate_4, output=""}
    False {from=MaterialExpressionComponentMask_6, output=""}
   ]
   properties {group=Blend A _Textures, parameterName=Tint Layer A}
  }
  MaterialExpressionStaticSwitch_0 {
   class=MaterialExpressionStaticSwitch
   inputs [
    True {from=MaterialExpressionSwitch_0, output=""}
    False {from=MaterialExpressionMultiply_9, output=""}
    Value {from=MaterialExpressionStaticBoolParameter_0, output=""}
   ]
   properties {}
  }
  MaterialExpressionStaticSwitch_1 {
   class=MaterialExpressionStaticSwitch
   inputs [
    True {from=MaterialExpressionStaticSwitch_22, output=""}
    False {from=MaterialExpressionStaticSwitch_4, output=""}
    Value {from=MaterialExpressionStaticBoolParameter_1, output=""}
   ]
   properties {}
  }
  MaterialExpressionStaticSwitch_10 {
   class=MaterialExpressionStaticSwitch
   inputs [
    True {from=MaterialExpressionScalarParameter_13, output=""}
    False {from=MaterialExpressionReroute_4, output=""}
    Value {from=MaterialExpressionStaticBoolParameter_11, output=""}
   ]
   properties {}
  }
  MaterialExpressionStaticSwitch_11 {
   class=MaterialExpressionStaticSwitch
   inputs [
    True {from=MaterialExpressionNamedRerouteUsage_57, output=""}
    False {from=MaterialExpressionConstant_1, output=""}
    Value {from=MaterialExpressionStaticBoolParameter_10, output=""}
   ]
   properties {}
  }
  MaterialExpressionStaticSwitch_12 {
   class=MaterialExpressionStaticSwitch
   inputs [
    True {from=MaterialExpressionScalarParameter_18, output=""}
    False {from=MaterialExpressionStaticSwitch_21, output=""}
    Value {from=MaterialExpressionStaticBoolParameter_12, output=""}
   ]
   properties {}
  }
  MaterialExpressionStaticSwitch_13 {
   class=MaterialExpressionStaticSwitch
   inputs [
    True {from=MaterialExpressionScalarParameter_19, output=""}
    False {from=MaterialExpressionStaticSwitch_20, output=""}
    Value {from=MaterialExpressionStaticBoolParameter_13, output=""}
   ]
   properties {}
  }
  MaterialExpressionStaticSwitch_14 {
   class=MaterialExpressionStaticSwitch
   inputs [
    True {from=MaterialExpressionOneMinus_2, output=""}
    False {from=MaterialExpressionMaterialFunctionCall_3, output=Result}
    Value {from=MaterialExpressionStaticBoolParameter_14, output=""}
   ]
   properties {}
  }
  MaterialExpressionStaticSwitch_15 {
   class=MaterialExpressionStaticSwitch
   inputs [
    True {from=MaterialExpressionMultiply_18, output=""}
    False {from=MaterialExpressionStaticSwitchParameter_1, output=""}
    Value {from=MaterialExpressionStaticBoolParameter_15, output=""}
   ]
   properties {}
  }
  MaterialExpressionStaticSwitch_16 {
   class=MaterialExpressionStaticSwitch
   inputs [
    True {from=MaterialExpressionNamedRerouteUsage_28, output=""}
    False {from=MaterialExpressionLinearInterpolate_6, output=""}
    Value {from=MaterialExpressionStaticBoolParameter_17, output=""}
   ]
   properties {}
  }
  MaterialExpressionStaticSwitch_17 {
   class=MaterialExpressionStaticSwitch
   inputs [
    True {from=MaterialExpressionNamedRerouteUsage_29, output=""}
    False {from=MaterialExpressionLinearInterpolate_1, output=""}
    Value {from=MaterialExpressionStaticBoolParameter_16, output=""}
   ]
   properties {}
  }
  MaterialExpressionStaticSwitch_18 {
   class=MaterialExpressionStaticSwitch
   inputs [
    True {from=MaterialExpressionStaticSwitch_7, output=""}
    False {from=MaterialExpressionNamedRerouteUsage_27, output=""}
    Value {from=MaterialExpressionStaticBoolParameter_19, output=""}
   ]
   properties {}
  }
  MaterialExpressionStaticSwitch_19 {
   class=MaterialExpressionStaticSwitch
   inputs [
    True {from=MaterialExpressionStaticSwitch_8, output=""}
    False {from=MaterialExpressionNamedRerouteUsage_22, output=""}
    Value {from=MaterialExpressionStaticBoolParameter_20, output=""}
   ]
   properties {}
  }
  MaterialExpressionStaticSwitch_2 {
   class=MaterialExpressionStaticSwitch
   inputs [
    True {from=MaterialExpressionStaticSwitch_1, output=""}
    False {from=MaterialExpressionConstant_0, output=""}
    Value {from=MaterialExpressionStaticBoolParameter_2, output=""}
   ]
   properties {}
  }
  MaterialExpressionStaticSwitch_20 {
   class=MaterialExpressionStaticSwitch
   inputs [
    True {from=MaterialExpressionMaterialFunctionCall_30, output=Result}
    False {from=MaterialExpressionReroute_5, output=""}
    Value {from=MaterialExpressionStaticBoolParameter_18, output=""}
   ]
   properties {}
  }
  MaterialExpressionStaticSwitch_21 {
   class=MaterialExpressionStaticSwitch
   inputs [
    True {from=MaterialExpressionMaterialFunctionCall_32, output=Result}
    False {from=MaterialExpressionReroute_6, output=""}
    Value {from=MaterialExpressionStaticBoolParameter_37, output=""}
   ]
   properties {}
  }
  MaterialExpressionStaticSwitch_22 {
   class=MaterialExpressionStaticSwitch
   inputs [
    True {from=MaterialExpressionMultiply_10, output=""}
    False {from=MaterialExpressionReroute_7, output=""}
    Value {from=MaterialExpressionStaticBoolParameter_38, output=""}
   ]
   properties {}
  }
  MaterialExpressionStaticSwitch_23 {
   class=MaterialExpressionStaticSwitch
   inputs [
    True {from=MaterialExpressionLinearInterpolate_7, output=""}
    False {from=MaterialExpressionNamedRerouteDeclaration_12, output=""}
    Value {from=MaterialExpressionStaticBoolParameter_39, output=""}
   ]
   properties {}
  }
  MaterialExpressionStaticSwitch_24 {
   class=MaterialExpressionStaticSwitch
   inputs [
    True {from=MaterialExpressionLinearInterpolate_8, output=""}
    False {from=MaterialExpressionNamedRerouteUsage_49, output=""}
    Value {from=MaterialExpressionStaticBoolParameter_40, output=""}
   ]
   properties {}
  }
  MaterialExpressionStaticSwitch_25 {
   class=MaterialExpressionStaticSwitch
   inputs [
    True {from=MaterialExpressionMaterialFunctionCall_28, output=Result}
    False {from=MaterialExpressionTextureSampleParameter2D_9, output=RGB}
    Value {from=MaterialExpressionStaticBoolParameter_41, output=""}
   ]
   properties {}
  }
  MaterialExpressionStaticSwitch_26 {
   class=MaterialExpressionStaticSwitch
   inputs [
    True {from=MaterialExpressionMaterialFunctionCall_33, output=Result}
    False {from=MaterialExpressionStaticSwitch_27, output=""}
    Value {from=MaterialExpressionStaticBoolParameter_42, output=""}
   ]
   properties {}
  }
  MaterialExpressionStaticSwitch_27 {
   class=MaterialExpressionStaticSwitch
   inputs [
    True {from=MaterialExpressionLinearInterpolate_9, output=""}
    False {from=MaterialExpressionStaticSwitch_24, output=""}
    Value {from=MaterialExpressionStaticBoolParameter_43, output=""}
   ]
   properties {}
  }
  MaterialExpressionStaticSwitch_28 {
   class=MaterialExpressionStaticSwitch
   inputs [
    True {from=MaterialExpressionDesaturation_0, output=""}
    False {from=MaterialExpressionStaticSwitch_25, output=""}
    Value {from=MaterialExpressionStaticBoolParameter_44, output=""}
   ]
   properties {}
  }
  MaterialExpressionStaticSwitch_29 {
   class=MaterialExpressionStaticSwitch
   inputs [
    True {from=MaterialExpressionAdd_19, output=""}
    False {from=MaterialExpressionReroute_10, output=""}
    Value {from=MaterialExpressionStaticBoolParameter_46, output=""}
   ]
   properties {}
  }
  MaterialExpressionStaticSwitch_3 {
   class=MaterialExpressionStaticSwitch
   inputs [
    True {from=MaterialExpressionLinearInterpolate_3, output=""}
    False {from=MaterialExpressionConstant3Vector_0, output=""}
    Value {from=MaterialExpressionStaticBoolParameter_3, output=""}
   ]
   properties {}
  }
  MaterialExpressionStaticSwitch_30 {
   class=MaterialExpressionStaticSwitch
   inputs [
    True {from=MaterialExpressionMultiply_24, output=""}
    False {from=MaterialExpressionReroute_11, output=""}
    Value {from=MaterialExpressionStaticBoolParameter_45, output=""}
   ]
   properties {}
  }
  MaterialExpressionStaticSwitch_4 {
   class=MaterialExpressionStaticSwitch
   inputs [
    True {from=MaterialExpressionMultiply_6, output=""}
    False {from=MaterialExpressionReroute_1, output=""}
    Value {from=MaterialExpressionStaticBoolParameter_4, output=""}
   ]
   properties {}
  }
  MaterialExpressionStaticSwitch_5 {
   class=MaterialExpressionStaticSwitch
   inputs [
    True {from=MaterialExpressionOneMinus_4, output=""}
    False {from=MaterialExpressionMaterialFunctionCall_0, output=Result}
    Value {from=MaterialExpressionStaticBoolParameter_5, output=""}
   ]
   properties {}
  }
  MaterialExpressionStaticSwitch_6 {
   class=MaterialExpressionStaticSwitch
   inputs [
    True {from=MaterialExpressionLinearInterpolate_0, output=""}
    False {from=MaterialExpressionNamedRerouteUsage_16, output=""}
    Value {from=MaterialExpressionStaticBoolParameter_7, output=""}
   ]
   properties {}
  }
  MaterialExpressionStaticSwitch_7 {
   class=MaterialExpressionStaticSwitch
   inputs [
    True {from=MaterialExpressionStaticSwitch_16, output=""}
    False {from=MaterialExpressionNamedRerouteUsage_26, output=""}
    Value {from=MaterialExpressionStaticBoolParameter_8, output=""}
   ]
   properties {}
  }
  MaterialExpressionStaticSwitch_8 {
   class=MaterialExpressionStaticSwitch
   inputs [
    True {from=MaterialExpressionStaticSwitch_17, output=""}
    False {from=MaterialExpressionNamedRerouteUsage_23, output=""}
    Value {from=MaterialExpressionStaticBoolParameter_9, output=""}
   ]
   properties {}
  }
  MaterialExpressionStaticSwitch_9 {
   class=MaterialExpressionStaticSwitch
   inputs [
    True {from=MaterialExpressionTextureSampleParameter2D_0, output=RGB}
    False {from=MaterialExpressionVectorParameter_0, output=RGB}
    Value {from=MaterialExpressionStaticBoolParameter_6, output=""}
   ]
   properties {}
  }
  MaterialExpressionStep_0 {
   class=MaterialExpressionStep
   inputs [
    Y {from=MaterialExpressionDistance_0, output=""}
    X {from=MaterialExpressionNamedRerouteUsage_56, output=""}
   ]
   properties {constX=0.25}
  }
  MaterialExpressionStep_1 {
   class=MaterialExpressionStep
   inputs [
    Y {from=MaterialExpressionDistance_1, output=""}
    X {from=MaterialExpressionNamedRerouteUsage_60, output=""}
   ]
   properties {constX=0.25}
  }
  MaterialExpressionStep_2 {
   class=MaterialExpressionStep
   inputs [
    Y {from=MaterialExpressionDistance_2, output=""}
    X {from=MaterialExpressionNamedRerouteUsage_61, output=""}
   ]
   properties {constX=0.25}
  }
  MaterialExpressionSubtract_0 {
   class=MaterialExpressionSubtract
   inputs [
    A {from=MaterialExpressionTextureSample_9, output=RGB}
    B {from=""}
   ]
   properties {constB=0.5}
  }
  MaterialExpressionSwitch_0 {
   class=MaterialExpressionSwitch
   inputs [
    SwitchValue {from=MaterialExpressionScalarParameter_63, output=""}
    Default {from=MaterialExpressionMaterialFunctionCall_9, output=Primary}
    None {from=MaterialExpressionMaterialFunctionCall_9, output=Primary}
    None {from=MaterialExpressionMaterialFunctionCall_9, output=Primary}
    None {from=MaterialExpressionMaterialFunctionCall_9, output=Primary}
   ]
   properties {inputs [
     {
      inputName=None
      input {
       expression=&self:MaterialExpressionMaterialFunctionCall_9
       inputName=None
       mask=0
       maskA=0
       maskB=0
       maskG=0
       maskR=0
       outputIndex=0
      }
     }
     {
      inputName=None
      input {
       expression=&self:MaterialExpressionMaterialFunctionCall_9
       inputName=None
       mask=0
       maskA=0
       maskB=0
       maskG=0
       maskR=0
       outputIndex=1
      }
     }
     {
      inputName=None
      input {
       expression=&self:MaterialExpressionMaterialFunctionCall_9
       inputName=None
       mask=0
       maskA=0
       maskB=0
       maskG=0
       maskR=0
       outputIndex=2
      }
     }
    ]}
  }
  MaterialExpressionSwitch_1 {
   class=MaterialExpressionSwitch
   inputs [
    SwitchValue {from=MaterialExpressionScalarParameter_66, output=""}
    Default {from=MaterialExpressionTextureCoordinate_2, output=""}
    None {from=MaterialExpressionTextureCoordinate_2, output=""}
    None {from=MaterialExpressionTextureCoordinate_3, output=""}
    None {from=MaterialExpressionTextureCoordinate_4, output=""}
    None {from=MaterialExpressionTextureCoordinate_5, output=""}
   ]
   properties {inputs [
     {
      inputName=None
      input {
       expression=&self:MaterialExpressionTextureCoordinate_2
       inputName=None
       mask=0
       maskA=0
       maskB=0
       maskG=0
       maskR=0
       outputIndex=0
      }
     }
     {
      inputName=None
      input {
       expression=&self:MaterialExpressionTextureCoordinate_3
       inputName=None
       mask=0
       maskA=0
       maskB=0
       maskG=0
       maskR=0
       outputIndex=0
      }
     }
     {
      inputName=None
      input {
       expression=&self:MaterialExpressionTextureCoordinate_4
       inputName=None
       mask=0
       maskA=0
       maskB=0
       maskG=0
       maskR=0
       outputIndex=0
      }
     }
     {
      inputName=None
      input {
       expression=&self:MaterialExpressionTextureCoordinate_5
       inputName=None
       mask=0
       maskA=0
       maskB=0
       maskG=0
       maskR=0
       outputIndex=0
      }
     }
    ]}
  }
  MaterialExpressionSwitch_2 {
   class=MaterialExpressionSwitch
   inputs [
    SwitchValue {from=MaterialExpressionScalarParameter_81, output=""}
    Default {from=MaterialExpressionTextureCoordinate_6, output=""}
    None {from=MaterialExpressionTextureCoordinate_6, output=""}
    None {from=MaterialExpressionTextureCoordinate_7, output=""}
    None {from=MaterialExpressionTextureCoordinate_8, output=""}
    None {from=MaterialExpressionTextureCoordinate_11, output=""}
   ]
   properties {inputs [
     {
      inputName=None
      input {
       expression=&self:MaterialExpressionTextureCoordinate_6
       inputName=None
       mask=0
       maskA=0
       maskB=0
       maskG=0
       maskR=0
       outputIndex=0
      }
     }
     {
      inputName=None
      input {
       expression=&self:MaterialExpressionTextureCoordinate_7
       inputName=None
       mask=0
       maskA=0
       maskB=0
       maskG=0
       maskR=0
       outputIndex=0
      }
     }
     {
      inputName=None
      input {
       expression=&self:MaterialExpressionTextureCoordinate_8
       inputName=None
       mask=0
       maskA=0
       maskB=0
       maskG=0
       maskR=0
       outputIndex=0
      }
     }
     {
      inputName=None
      input {
       expression=&self:MaterialExpressionTextureCoordinate_11
       inputName=None
       mask=0
       maskA=0
       maskB=0
       maskG=0
       maskR=0
       outputIndex=0
      }
     }
    ]}
  }
  MaterialExpressionTextureCoordinate_0 {class=MaterialExpressionTextureCoordinate}
  MaterialExpressionTextureCoordinate_1 {class=MaterialExpressionTextureCoordinate, properties {coordinateIndex=1}}
  MaterialExpressionTextureCoordinate_10 {class=MaterialExpressionTextureCoordinate}
  MaterialExpressionTextureCoordinate_11 {class=MaterialExpressionTextureCoordinate, properties {coordinateIndex=3}}
  MaterialExpressionTextureCoordinate_2 {class=MaterialExpressionTextureCoordinate}
  MaterialExpressionTextureCoordinate_3 {class=MaterialExpressionTextureCoordinate, properties {coordinateIndex=1}}
  MaterialExpressionTextureCoordinate_4 {class=MaterialExpressionTextureCoordinate, properties {coordinateIndex=2}}
  MaterialExpressionTextureCoordinate_5 {class=MaterialExpressionTextureCoordinate, properties {coordinateIndex=3}}
  MaterialExpressionTextureCoordinate_6 {class=MaterialExpressionTextureCoordinate}
  MaterialExpressionTextureCoordinate_7 {class=MaterialExpressionTextureCoordinate, properties {coordinateIndex=1}}
  MaterialExpressionTextureCoordinate_8 {class=MaterialExpressionTextureCoordinate, properties {coordinateIndex=2}}
  MaterialExpressionTextureCoordinate_9 {class=MaterialExpressionTextureCoordinate}
  MaterialExpressionTextureObjectParameter_0 {
   class=MaterialExpressionTextureObjectParameter
   inputs [
    UVs {from="", output=""}
    Apply View MipBias {from="", output=""}
   ]
   properties {group=Blend _Mask, parameterName=_Mask, samplerType=SAMPLERTYPE_Masks, texture=/Game/Global/Textures/Default/Blanks/T_MidgreyMask}
  }
  MaterialExpressionTextureObjectParameter_1 {
   class=MaterialExpressionTextureObjectParameter
   inputs [
    UVs {from="", output=""}
    Apply View MipBias {from="", output=""}
   ]
   properties {
    group=Blend A _Textures
    parameterName=Layer A Diffuse
    samplerType=SAMPLERTYPE_LinearColor
    texture=/SparksMaps/Environment/Green_Room/Textures/T_GR_PlasterWall_D
   }
  }
  MaterialExpressionTextureObjectParameter_11 {
   class=MaterialExpressionTextureObjectParameter
   inputs [
    UVs {from="", output=""}
    Apply View MipBias {from="", output=""}
   ]
   properties {group=06 Glitter, parameterName=__Glitter Noise, texture=/SparksMaps/Environment/S5/Generic/Textures/T_Sparkle_Spots_B}
  }
  MaterialExpressionTextureObjectParameter_2 {
   class=MaterialExpressionTextureObjectParameter
   inputs [
    UVs {from="", output=""}
    Apply View MipBias {from="", output=""}
   ]
   properties {
    group=Blend A _Textures
    parameterName=Layer A Masks
    samplerType=SAMPLERTYPE_Masks
    texture=&T_MidgreyMask
   }
  }
  MaterialExpressionTextureObjectParameter_3 {
   class=MaterialExpressionTextureObjectParameter
   inputs [
    UVs {from="", output=""}
    Apply View MipBias {from="", output=""}
   ]
   properties {
    group=Blend A _Textures
    parameterName=Layer A Normal
    samplerType=SAMPLERTYPE_Normal
    texture=&FlatNormalBlue
   }
  }
  MaterialExpressionTextureObjectParameter_4 {
   class=MaterialExpressionTextureObjectParameter
   inputs [
    UVs {from="", output=""}
    Apply View MipBias {from="", output=""}
   ]
   properties {
    group=Blend B _Textures
    parameterName=Layer B Normal
    samplerType=SAMPLERTYPE_Normal
    texture=&FlatNormalBlue
   }
  }
  MaterialExpressionTextureObjectParameter_5 {
   class=MaterialExpressionTextureObjectParameter
   inputs [
    UVs {from="", output=""}
    Apply View MipBias {from="", output=""}
   ]
   properties {
    group=Blend B _Textures
    parameterName=Layer B Masks
    samplerType=SAMPLERTYPE_Masks
    texture=&T_MidgreyMask
   }
  }
  MaterialExpressionTextureObjectParameter_6 {
   class=MaterialExpressionTextureObjectParameter
   inputs [
    UVs {from="", output=""}
    Apply View MipBias {from="", output=""}
   ]
   properties {group=Blend B _Textures, parameterName=Layer B Diffuse, texture=/SparksMaps/Environment/Green_Room/Textures/T_GR_Marble_Tile_A_D}
  }
  MaterialExpressionTextureObjectParameter_7 {
   class=MaterialExpressionTextureObjectParameter
   inputs [
    UVs {from="", output=""}
    Apply View MipBias {from="", output=""}
   ]
   properties {group=08 Halftone, parameterName=Halftone Mask, sortPriority=4}
  }
  MaterialExpressionTextureSampleParameter2D_0 {
   class=MaterialExpressionTextureSampleParameter2D
   inputs [
    UVs {from=MaterialExpressionMaterialFunctionCall_6, output=Result}
    Apply View MipBias {from="", output=""}
   ]
   properties {
    group=01 Diffuse_Fresnel
    parameterName=FresnelDiffuse_Texture
    samplerType=SAMPLERTYPE_Masks
    sortPriority=10
    texture=&T_MidgreyMask
   }
  }
  MaterialExpressionTextureSampleParameter2D_1 {
   class=MaterialExpressionTextureSampleParameter2D
   inputs [
    UVs {from=MaterialExpressionMaterialFunctionCall_5, output=Result}
    Apply View MipBias {from="", output=""}
   ]
   properties {
    group=03 Diffuse Multiply Pattern
    parameterName=Diffuse_Pattern_Texture
    samplerType=SAMPLERTYPE_Masks
    texture=&T_MidgreyMask
   }
  }
  MaterialExpressionTextureSampleParameter2D_2 {
   class=MaterialExpressionTextureSampleParameter2D
   inputs [
    UVs {from=MaterialExpressionMaterialFunctionCall_10, output=Result}
    Apply View MipBias {from="", output=""}
   ]
   properties {
    group=01 _Textures
    parameterName=_Detail Normals
    samplerType=SAMPLERTYPE_Normal
    texture=&FlatNormalBlue
   }
  }
  MaterialExpressionTextureSampleParameter2D_3 {
   class=MaterialExpressionTextureSampleParameter2D
   inputs [
    UVs {from=MaterialExpressionNamedRerouteUsage_5, output=""}
    Apply View MipBias {from="", output=""}
   ]
   properties {
    group=01 Diffuse_Fresnel
    parameterName=Diffuse_Fresnel Orerride Masks
    samplerType=SAMPLERTYPE_Masks
    texture=&T_MidgreyMask
    channelNames {a=Unused, b=Roughness, g=Metallic, r=Specular}
   }
  }
  MaterialExpressionTextureSampleParameter2D_4 {
   class=MaterialExpressionTextureSampleParameter2D
   inputs [
    UVs {from=MaterialExpressionNamedRerouteUsage_2, output=""}
    Apply View MipBias {from="", output=""}
   ]
   properties {
    group=01 _Textures
    parameterName=__PBR Masks
    samplerType=SAMPLERTYPE_Masks
    texture=&T_MidgreyMask
    channelNames {a=Unused, b=Roughness, g=Metallic, r=Specular}
   }
  }
  MaterialExpressionTextureSampleParameter2D_5 {
   class=MaterialExpressionTextureSampleParameter2D
   inputs [
    UVs {from=MaterialExpressionNamedRerouteUsage_0, output=""}
    Apply View MipBias {from="", output=""}
   ]
   properties {group=01 _Textures, parameterName=_Normal, samplerType=SAMPLERTYPE_Normal, texture=&FlatNormalBlue}
  }
  MaterialExpressionTextureSampleParameter2D_6 {
   class=MaterialExpressionTextureSampleParameter2D
   inputs [
    UVs {from=MaterialExpressionNamedRerouteUsage_1, output=""}
    Apply View MipBias {from="", output=""}
   ]
   properties {group=01 _Textures, parameterName=___Diffuse, samplerType=SAMPLERTYPE_Masks, texture=&T_MidgreyMask}
  }
  MaterialExpressionTextureSampleParameter2D_7 {
   class=MaterialExpressionTextureSampleParameter2D
   inputs [
    UVs {from=MaterialExpressionMaterialFunctionCall_7, output=Result}
    Apply View MipBias {from="", output=""}
   ]
   properties {group=05 Emissive, parameterName=Emiss, samplerType=SAMPLERTYPE_Masks, texture=&T_MidgreyMask}
  }
  MaterialExpressionTextureSampleParameter2D_8 {
   class=MaterialExpressionTextureSampleParameter2D
   inputs [
    UVs {from=MaterialExpressionMaterialFunctionCall_64, output=Result}
    Apply View MipBias {from="", output=""}
   ]
   properties {
    group=06 Glitter
    parameterName=Glitter Mask
    texture=/SparksMaps/Environment/S5/Venue_A/Sets/Main_Stage_Deco/Textures/T_Main_Stage_Glitter_A_M
   }
  }
  MaterialExpressionTextureSampleParameter2D_9 {
   class=MaterialExpressionTextureSampleParameter2D
   inputs [
    UVs {from=MaterialExpressionSwitch_1, output=""}
    Apply View MipBias {from="", output=""}
   ]
   properties {group=07 Graffiti, parameterName=Graffiti Texture, sortPriority=1}
  }
  MaterialExpressionTextureSampleParameterCube_0 {
   class=MaterialExpressionTextureSampleParameterCube
   inputs [
    UVs {from=MaterialExpressionReflectionVectorWS_0, output=""}
    Apply View MipBias {from="", output=""}
   ]
   properties {group=05 Emissive, parameterName=CubeMap}
  }
  MaterialExpressionTextureSample_0 {
   class=MaterialExpressionTextureSample
   inputs [
    UVs {from=MaterialExpressionNamedRerouteUsage_10}
    Tex {from=MaterialExpressionTextureObjectParameter_1}
    Apply View MipBias {from=""}
   ]
  }
  MaterialExpressionTextureSample_1 {
   class=MaterialExpressionTextureSample
   inputs [
    UVs {from=MaterialExpressionNamedRerouteUsage_33}
    Tex {from=MaterialExpressionTextureObjectParameter_2}
    Apply View MipBias {from=""}
   ]
  }
  MaterialExpressionTextureSample_16 {
   class=MaterialExpressionTextureSample
   inputs [
    UVs {from=MaterialExpressionMultiply_50}
    Tex {from=MaterialExpressionNamedRerouteDeclaration_19}
    Apply View MipBias {from=""}
   ]
  }
  MaterialExpressionTextureSample_17 {
   class=MaterialExpressionTextureSample
   inputs [
    UVs {from=MaterialExpressionMultiply_55}
    Tex {from=MaterialExpressionNamedRerouteDeclaration_19}
    Apply View MipBias {from=""}
   ]
  }
  MaterialExpressionTextureSample_2 {
   class=MaterialExpressionTextureSample
   inputs [
    UVs {from=MaterialExpressionNamedRerouteUsage_34}
    Tex {from=MaterialExpressionTextureObjectParameter_3}
    Apply View MipBias {from=""}
   ]
  }
  MaterialExpressionTextureSample_3 {
   class=MaterialExpressionTextureSample
   inputs [
    UVs {from=MaterialExpressionNamedRerouteUsage_35}
    Tex {from=MaterialExpressionTextureObjectParameter_6}
    Apply View MipBias {from=""}
   ]
  }
  MaterialExpressionTextureSample_4 {
   class=MaterialExpressionTextureSample
   inputs [
    UVs {from=MaterialExpressionNamedRerouteUsage_36}
    Tex {from=MaterialExpressionTextureObjectParameter_5}
    Apply View MipBias {from=""}
   ]
  }
  MaterialExpressionTextureSample_5 {
   class=MaterialExpressionTextureSample
   inputs [
    UVs {from=MaterialExpressionNamedRerouteUsage_37}
    Tex {from=MaterialExpressionTextureObjectParameter_4}
    Apply View MipBias {from=""}
   ]
  }
  MaterialExpressionTextureSample_6 {
   class=MaterialExpressionTextureSample
   inputs [
    UVs {from=MaterialExpressionMaterialFunctionCall_13, output=Result}
    Tex {from=MaterialExpressionNamedRerouteDeclaration_29}
    Apply View MipBias {from=""}
   ]
  }
  MaterialExpressionTextureSample_7 {
   class=MaterialExpressionTextureSample
   inputs [
    UVs {from=MaterialExpressionMaterialFunctionCall_4, output=Result}
    Tex {from=MaterialExpressionNamedRerouteUsage_46}
    Apply View MipBias {from=""}
   ]
  }
  MaterialExpressionTextureSample_8 {
   class=MaterialExpressionTextureSample
   inputs [
    UVs {from=MaterialExpressionMultiply_29}
    Tex {from=MaterialExpressionTextureObjectParameter_7}
    Apply View MipBias {from=""}
   ]
  }
  MaterialExpressionTextureSample_9 {
   class=MaterialExpressionTextureSample
   inputs [
    UVs {from=MaterialExpressionMultiply_0}
    Tex {from=MaterialExpressionNamedRerouteUsage_31}
    Apply View MipBias {from=""}
   ]
  }
  MaterialExpressionVectorParameter_0 {
   class=MaterialExpressionVectorParameter
   properties {group=01 Diffuse_Fresnel, parameterName=DiffuseFresnelColor, sortPriority=9, defaultValue=<rgba 1 1 1 1>}
  }
  MaterialExpressionVectorParameter_1 {
   class=MaterialExpressionVectorParameter
   properties {
    group=Blend A Fine Tune Edge
    parameterName="Fine Tune Edge (Blend A)"
    channelNames {a=Edge High End, b=Edge Low End, g=Body High End, r=Body Low End}
    defaultValue=<rgba -15 150 -15 120>
   }
  }
  MaterialExpressionVectorParameter_2 {
   class=MaterialExpressionVectorParameter
   properties {group=01 _Textures, parameterName=BaseColor_Tint_Intensity, sortPriority=36, defaultValue=<rgba 1 1 1 1>}
  }
  MaterialExpressionVectorParameter_3 {
   class=MaterialExpressionVectorParameter
   properties {group=05 Emissive, parameterName=Emissive_Tint, sortPriority=34, defaultValue=<rgba 1 1 1 1>}
  }
  MaterialExpressionVectorParameter_4 {
   class=MaterialExpressionVectorParameter
   properties {
    group=Blend B Fine Tune Edge
    parameterName="Fine Tune Edge (Blend B)"
    channelNames {a=Edge High End, b=Edge Low End, g=Body High End, r=Body Low End}
    defaultValue=<rgba -15 150 -15 120>
   }
  }
  MaterialExpressionVectorParameter_5 {class=MaterialExpressionVectorParameter, properties {group=Blend A _Textures, parameterName="Tint (Layer A)"}}
  MaterialExpressionVectorParameter_6 {class=MaterialExpressionVectorParameter, properties {group=Blend B _Textures, parameterName="Tint (Layer B)"}}
  MaterialExpressionVectorParameter_7 {
   class=MaterialExpressionVectorParameter
   properties {group=08 Halftone, parameterName=Halftone Color A, sortPriority=5, defaultValue=<rgba 1 0.2264 0.2689 1>}
  }
  MaterialExpressionVectorParameter_8 {
   class=MaterialExpressionVectorParameter
   properties {group=08 Halftone, parameterName=Halftone Color B, sortPriority=6, defaultValue=<rgba 0.0502 1 0 1>}
  }
  MaterialExpressionVectorParameter_9 {
   class=MaterialExpressionVectorParameter
   properties {group=08 Halftone, parameterName=Halftone Color C, sortPriority=7, defaultValue=<rgba 0.3508 0 1 1>}
  }
  MaterialExpressionVertexColor_0 {class=MaterialExpressionVertexColor, inputs [], properties {}}
  MaterialExpressionVertexColor_3 {class=MaterialExpressionVertexColor, inputs [], properties {}}
  MaterialExpressionVertexNormalWS_0 {class=MaterialExpressionVertexNormalWS, inputs [], properties {}}
  MaterialExpressionVertexNormalWS_1 {class=MaterialExpressionVertexNormalWS, inputs [], properties {}}
 }
 properties {bAutomaticallySetUsageInEditor=false, bUseMaterialAttributes=true, bUsedWithNanite=true, usedShadingModels=Default Lit}
}
